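# Physics asset for the F_MED_RoosterMelt female jacket, defining capsule bodies and joints for secondary motion on its dynamic jacket bones.
# Six chains (left and right front, side and back) each link a zero-size base body to parent and tip capsules through swing-limited constraints.
# Swing limits generally widen toward the tips, with twist zeroed on all but the front-left root joint.
# Two unconstrained upper-spine capsules provide torso volumes.
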PhysicsAsset /BRCosmetics/Characters/Player/Female/Medium/Bodies/F_MED_RoosterMelt/Meshes/F_MED_RoosterMelt_Jacket_Physics
bodies {
 dyn_jacket_side_bk_l_01 [{
   shapeType=Capsule
   length=10.839
   radius=0.7555
   shapeName=dyn_jacket_side_bk_l_01_capsule
   center=<-1.6285 0 0>
   rotation=<rot -90 179.9802 -179.9803>
  }]
 dyn_jacket_side_bk_l_base [{shapeType=Capsule, length=0, radius=0, shapeName=dyn_jacket_side_bk_l_base_capsule}]
 dyn_jacket_side_bk_l_parent [{shapeType=Capsule, length=8.8711, radius=3.3496, shapeName=dyn_jacket_side_bk_l_parent_capsule}]
 dyn_jacket_side_bk_r_01 [{
   shapeType=Capsule
   length=10.4968
   radius=1.93
   shapeName=dyn_jacket_side_bk_r_01_capsule
   center=<1.811 0 0>
   rotation=<rot 90 179.9703 179.9704>
  }]
 dyn_jacket_side_bk_r_base [{shapeType=Capsule, length=0, radius=0, shapeName=dyn_jacket_side_bk_r_base_capsule}]
 dyn_jacket_side_bk_r_parent [{shapeType=Capsule, length=8.25, radius=3.505, shapeName=dyn_jacket_side_bk_r_parent_capsule}]
 dyn_jacket_side_frnt_l_01 [{
   shapeType=Capsule
   length=5.6127
   radius=3.3093
   shapeName=dyn_jacket_side_frnt_l_01_capsule
   center=<3.1731 1.97 0>
   rotation=<rot 90 0.0001 0>
  }]
 dyn_jacket_side_frnt_l_base [{shapeType=Capsule, length=0, radius=0, shapeName=dyn_jacket_side_frnt_l_base_capsule, rotation=<rot 0 -90 0>}]
 dyn_jacket_side_frnt_l_parent [{
   shapeType=Capsule
   length=5.0731
   radius=2.736
   shapeName=dyn_jacket_side_frnt_l_parent_capsule
   center=<2.9767 1.7625 0>
   rotation=<rot 90 0 -0.0001>
  }]
 dyn_jacket_side_frnt_r_01 [{
   shapeType=Capsule
   length=5.5
   radius=3.2186
   shapeName=dyn_jacket_side_frnt_r_01_capsule
   center=<-3.5944 -2.0144 0>
   rotation=<rot 90 -179.9999 179.9997>
  }]
 dyn_jacket_side_frnt_r_base [{shapeType=Capsule, length=0, radius=0}]
 dyn_jacket_side_frnt_r_parent [{
   shapeType=Capsule
   length=4.75
   radius=2.755
   shapeName=dyn_jacket_side_frnt_r_parent_capsule
   center=<-3.4719 -1.8274 0>
   rotation=<rot -90 179.9605 -179.9604>
  }]
 dyn_jacket_side_l_01 [{shapeType=Capsule, length=7.4987, radius=0.9899, shapeName=dyn_jacket_side_l_01_capsule, center=<4.1123 0 0>, rotation=<rot 90 0 0>}]
 dyn_jacket_side_l_base [{shapeType=Capsule, length=0, radius=0, shapeName=dyn_jacket_side_l_base_capsule, rotation=<rot 0 -90 0>}]
 dyn_jacket_side_l_parent [{shapeType=Capsule, length=4.755, radius=2.005, shapeName=dyn_jacket_side_l_parent_capsule, rotation=<rot 90 0 0>}]
 dyn_jacket_side_r_01 [{
   shapeType=Capsule
   length=6.1527
   radius=1.815
   shapeName=dyn_jacket_side_r_01_capsule
   center=<-3.7403 0 0>
   rotation=<rot 90 -179.9802 -179.9802>
  }]
 dyn_jacket_side_r_base [{shapeType=Capsule, length=0, radius=0, shapeName=dyn_jacket_side_r_base_capsule}]
 dyn_jacket_side_r_parent [{shapeType=Capsule, length=3.5845, radius=2.0757, shapeName=dyn_jacket_side_r_parent_capsule, rotation=<rot 90 -179.9703 -179.9703>}]
 spine_03 [{shapeType=Capsule, length=5.046, radius=8.4924, center=<0 1.4778 0>}]
 spine_04 [{shapeType=Capsule, length=6.7695, radius=8.7661, shapeName=spine_04_capsule, center=<0 1.7028 0>}]
}
constraints [
 {bone1Name=dyn_jacket_side_bk_r_parent, bone2Name=dyn_jacket_side_bk_r_base, swing1LimitDegrees=2, swing2LimitDegrees=10, twistLimitDegrees=0}
 {bone1Name=dyn_jacket_side_bk_r_01, bone2Name=dyn_jacket_side_bk_r_parent, swing1LimitDegrees=5, swing2LimitDegrees=20, twistLimitDegrees=0}
 {
  bone1Name=dyn_jacket_side_frnt_l_01
  bone2Name=dyn_jacket_side_frnt_l_parent
  swing1LimitDegrees=10
  swing2LimitDegrees=30
  twistLimitDegrees=0
 }
 {
  bone1Name=dyn_jacket_side_frnt_r_parent
  bone2Name=dyn_jacket_side_frnt_r_base
  swing1LimitDegrees=5
  swing2LimitDegrees=5
  twistLimitDegrees=0
 }
 {
  bone1Name=dyn_jacket_side_frnt_r_01
  bone2Name=dyn_jacket_side_frnt_r_parent
  swing1LimitDegrees=10
  swing2LimitDegrees=30
  twistLimitDegrees=0
 }
 {bone1Name=dyn_jacket_side_l_parent, bone2Name=dyn_jacket_side_l_base, swing1LimitDegrees=5, swing2LimitDegrees=5, twistLimitDegrees=0}
 {bone1Name=dyn_jacket_side_l_01, bone2Name=dyn_jacket_side_l_parent, swing1LimitDegrees=10, swing2LimitDegrees=20, twistLimitDegrees=0}
 {bone1Name=dyn_jacket_side_r_parent, bone2Name=dyn_jacket_side_r_base, swing1LimitDegrees=3, swing2LimitDegrees=0, twistLimitDegrees=0}
 {bone1Name=dyn_jacket_side_r_01, bone2Name=dyn_jacket_side_r_parent, swing1LimitDegrees=10, swing2LimitDegrees=20, twistLimitDegrees=0}
 {
  bone1Name=dyn_jacket_side_frnt_l_parent
  bone2Name=dyn_jacket_side_frnt_l_base
  swing1LimitDegrees=5
  swing2LimitDegrees=5
  twistLimitDegrees=5
 }
 {bone1Name=dyn_jacket_side_bk_l_01, bone2Name=dyn_jacket_side_bk_l_parent, swing1LimitDegrees=20, swing2LimitDegrees=20, twistLimitDegrees=0}
 {bone1Name=dyn_jacket_side_bk_l_parent, bone2Name=dyn_jacket_side_bk_l_base, swing1LimitDegrees=5, swing2LimitDegrees=5, twistLimitDegrees=0}
]
