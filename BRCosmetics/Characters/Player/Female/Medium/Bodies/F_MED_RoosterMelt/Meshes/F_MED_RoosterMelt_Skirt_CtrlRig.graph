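# Control Rig for the RoosterMelt female body's skirt, posing the dyn_skirt bones from leg motion and movement state.
# Spherical pose readers on each thigh weight parent constraints and corrective rotations on side, front and back bones.
# Mid front and back bones average their left and right neighbours, with extra branches for skydive, tactical sprint and baseball slide.
# Foot proximity, crouch state and directional movement weights add further offsets.
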ControlRigBlueprint /BRCosmetics/Characters/Player/Female/Medium/Bodies/F_MED_RoosterMelt/Meshes/F_MED_RoosterMelt_Skirt_CtrlRig
graphs [
 RigVMFunctionLibrary {links [], nodes []}
 RigVMModel {nodes [
   RigUnit_BeginExecution = &RigUnit_BeginExecution()
   SequenceAggregate_1 = RigVMAggregateNode(ExecuteContext=@RigUnit_BeginExecution.ExecutePin)
   n1 = &RigUnit_SphericalPoseReader(
    ActiveRegionScaleFactors=(PositiveWidth=3.5,NegativeWidth=3,PositiveHeight=3,NegativeHeight=2)
    ActiveRegionSize=0.1
    Debug=(bDrawDebug=false,DebugScale=25,DebugSegments=20,DebugThickness=0.25)
    DriverAxis=(X=-1,Y=0,Z=0)
    DriverItem=(Type=Bone,Name="thigh_l")
    FalloffRegionScaleFactors=(PositiveWidth=1,NegativeWidth=2.4,PositiveHeight=2.4,NegativeHeight=2)
    FalloffSize=0.5
    OptionalParentItem=(Type=Bone)
    RotationOffset=(X=110,Y=0,Z=0)
    ExecutePin=@SequenceAggregate_1.A
   )
   > &RigUnit_ParentConstraint(
    AdvancedSettings=(RotationOrderForFilter=XZY)
    Child=(Type=Bone,Name="dyn_skirt_side_frnt_l_base")
    Filter=(TranslationFilter=(bX=false,bY=false,bZ=false),RotationFilter=(bX=True,bY=True,bZ=True),ScaleFilter=(bX=True,bY=True,bZ=True))
    Parents=((Item=(Type=Bone,Name="thigh_l"),Weight=0.8),(Item=(Type=Bone,Name="pelvis"),Weight=0.2))
    Weight=@n1.OutputParam
   )
   > &RigUnit_ParentConstraint(
    AdvancedSettings=(RotationOrderForFilter=XZY)
    Child=(Type=Bone,Name="dyn_skirt_side_l_base")
    Filter=(TranslationFilter=(bX=false,bY=false,bZ=false),RotationFilter=(bX=True,bY=True,bZ=True),ScaleFilter=(bX=True,bY=True,bZ=True))
    Parents=((Item=(Type=Bone,Name="thigh_l"),Weight=0.4),(Item=(Type=Bone,Name="pelvis"),Weight=0.6))
    Weight=@n1.OutputParam
   )
   > n2 = &RigUnit_SphericalPoseReader(
    ActiveRegionScaleFactors=(PositiveWidth=3.5,NegativeWidth=1,PositiveHeight=3,NegativeHeight=3)
    ActiveRegionSize=0.08
    Debug=(bDrawDebug=false,DebugScale=25,DebugSegments=20,DebugThickness=0.25)
    DriverAxis=(X=-1,Y=0,Z=0)
    DriverItem=(Type=Bone,Name="thigh_l")
    FalloffRegionScaleFactors=(PositiveWidth=1,NegativeWidth=1.5,PositiveHeight=1.6,NegativeHeight=1)
    FalloffSize=0.5
    OptionalParentItem=(Type=Bone)
    RotationOffset=(X=110,Y=0,Z=-40)
   )
   > &RigUnit_ModifyTransforms(
    ItemToModify=((Item=(Type=Bone,Name="dyn_skirt_side_l_base"),Transform=(Rotation=(X=0,Y=0,Z=0,W=1),Translation=(X=0,Y=-3,Z=0),Scale3D=(X=1,Y=1,Z=1))))
    ItemToModify.0.Transform.Rotation=&RigVMFunction_MathQuaternionFromEuler(Euler=(X=0,Y=-25,Z=20)).Result
    Weight=@n2.OutputParam
   )
   > &RigUnit_ModifyTransforms(
    ItemToModify=((Item=(Type=Bone,Name="dyn_skirt_side_frnt_l_base"),Transform=(Rotation=(X=0,Y=0,Z=0,W=1),Translation=(X=-2,Y=0,Z=3),Scale3D=(X=1,Y=1,Z=1))))
    ItemToModify.0.Transform.Rotation=&RigVMFunction_MathQuaternionFromEuler(Euler=(X=0,Y=10,Z=12)).Result
    Weight=@n2.OutputParam
   )
   > n3 = &RigUnit_SphericalPoseReader(
    ActiveRegionScaleFactors=(PositiveWidth=1,NegativeWidth=1,PositiveHeight=0.6,NegativeHeight=0.6)
    ActiveRegionSize=0.3
    Debug=(bDrawDebug=true,DebugScale=25,DebugSegments=20,DebugThickness=0.25)
    DriverAxis=(X=-1,Y=0,Z=0)
    DriverItem=(Type=Bone,Name="thigh_l")
    FalloffRegionScaleFactors=(PositiveWidth=1,NegativeWidth=1.5,PositiveHeight=1,NegativeHeight=0.3)
    FalloffSize=0.5
    FlipHeightScaling=false
    FlipWidthScaling=false
    OptionalParentItem=(Type=Bone)
    RotationOffset=(X=180,Y=0,Z=0)
   )
   > &RigUnit_ModifyTransforms(
    ItemToModify=((Item=(Type=Bone,Name="dyn_skirt_side_l_01"),Transform=(Rotation=(X=0,Y=0,Z=0,W=1),Translation=(X=0,Y=0,Z=0),Scale3D=(X=1,Y=1,Z=1))))
    ItemToModify.0.Transform.Rotation=&RigVMFunction_MathQuaternionFromEuler(Euler=(X=0,Y=-30,Z=15)).Result
    Weight=@n3.OutputParam
   )
   > &RigUnit_ModifyTransforms(
    ItemToModify=((Item=(Type=Bone,Name="dyn_skirt_side_frnt_l_01"),Transform=(Rotation=(X=0,Y=0,Z=0,W=1),Translation=(X=-2,Y=0,Z=0),Scale3D=(X=1,Y=1,Z=1))))
    ItemToModify.0.Transform.Rotation=&RigVMFunction_MathQuaternionFromEuler(Euler=(X=0,Y=8,Z=25)).Result
    Weight=@n3.OutputParam
   )
   > &RigUnit_ModifyTransforms(
    ItemToModify=((Item=(Type=Bone,Name="dyn_skirt_side_bk_l_base"),Transform=(Rotation=(X=0,Y=0,Z=0,W=1),Translation=(X=0,Y=0,Z=0),Scale3D=(X=1,Y=1,Z=1))))
    ItemToModify.0.Transform.Rotation=&RigVMFunction_MathQuaternionFromEuler(Euler=(X=0,Y=5,Z=-15)).Result
    Weight=@n3.OutputParam
   )
   n4 = &RigUnit_SphericalPoseReader(
    ActiveRegionScaleFactors=(PositiveWidth=3.5,NegativeWidth=3,PositiveHeight=2,NegativeHeight=1)
    ActiveRegionSize=0.1
    Debug=(bDrawDebug=false,DebugScale=25,DebugSegments=20,DebugThickness=0.25)
    DriverAxis=(X=1,Y=0,Z=0)
    DriverItem=(Type=Bone,Name="thigh_r")
    FalloffRegionScaleFactors=(PositiveWidth=0.9,NegativeWidth=2,PositiveHeight=1,NegativeHeight=0.9)
    FalloffSize=1
    FlipHeightScaling=false
    FlipWidthScaling=true
    OptionalParentItem=(Type=Bone)
    RotationOffset=(X=-70,Y=0,Z=20)
    ExecutePin=@SequenceAggregate_1.B
   )
   > &RigUnit_ParentConstraint(
    AdvancedSettings=(RotationOrderForFilter=XZY)
    Child=(Type=Bone,Name="dyn_skirt_side_frnt_r_base")
    Filter=(TranslationFilter=(bX=false,bY=false,bZ=false),RotationFilter=(bX=True,bY=True,bZ=True),ScaleFilter=(bX=True,bY=True,bZ=True))
    Parents=((Item=(Type=Bone,Name="thigh_r"),Weight=0.8),(Item=(Type=Bone,Name="pelvis"),Weight=0.2))
    Weight=@n4.OutputParam
   )
   > &RigUnit_ParentConstraint(
    AdvancedSettings=(RotationOrderForFilter=XZY)
    Child=(Type=Bone,Name="dyn_skirt_side_r_base")
    Filter=(TranslationFilter=(bX=false,bY=false,bZ=false),RotationFilter=(bX=True,bY=True,bZ=True),ScaleFilter=(bX=True,bY=True,bZ=True))
    Parents=((Item=(Type=Bone,Name="thigh_r"),Weight=0.4),(Item=(Type=Bone,Name="pelvis"),Weight=0.6))
    Weight=@n4.OutputParam
   )
   > n5 = &RigUnit_SphericalPoseReader(
    ActiveRegionScaleFactors=(PositiveWidth=3.5,NegativeWidth=1,PositiveHeight=3,NegativeHeight=3)
    ActiveRegionSize=0.1
    Debug=(bDrawDebug=true,DebugScale=25,DebugSegments=20,DebugThickness=0.25)
    DriverAxis=(X=1,Y=0,Z=0)
    DriverItem=(Type=Bone,Name="thigh_r")
    FalloffRegionScaleFactors=(PositiveWidth=1,NegativeWidth=1.4,PositiveHeight=1.1,NegativeHeight=1)
    FalloffSize=0.5
    FlipHeightScaling=false
    FlipWidthScaling=true
    OptionalParentItem=(Type=Bone)
    RotationOffset=(X=-70,Y=0,Z=-20)
   )
   > &RigUnit_ModifyTransforms(
    ItemToModify=((Item=(Type=Bone,Name="dyn_skirt_side_r_base"),Transform=(Rotation=(X=0,Y=0,Z=0,W=1),Translation=(X=0,Y=0,Z=0),Scale3D=(X=1,Y=1,Z=1))))
    ItemToModify.0.Transform.Rotation=&RigVMFunction_MathQuaternionFromEuler(Euler=(X=0,Y=-25,Z=20)).Result
    Weight=@n5.OutputParam
   )
   > &RigUnit_ModifyTransforms(
    ItemToModify=((Item=(Type=Bone,Name="dyn_skirt_side_frnt_r_base"),Transform=(Rotation=(X=0,Y=0,Z=0,W=1),Translation=(X=0,Y=0,Z=0),Scale3D=(X=1,Y=1,Z=1))))
    ItemToModify.0.Transform.Rotation=&RigVMFunction_MathQuaternionFromEuler(Euler=(X=0,Y=10,Z=15)).Result
    Weight=@n5.OutputParam
   )
   > n6 = &RigUnit_SphericalPoseReader(
    ActiveRegionScaleFactors=(PositiveWidth=1,NegativeWidth=1,PositiveHeight=1,NegativeHeight=0.6)
    ActiveRegionSize=0.3
    Debug=(bDrawDebug=true,DebugScale=25,DebugSegments=20,DebugThickness=0.25)
    DriverAxis=(X=1,Y=0,Z=0)
    DriverItem=(Type=Bone,Name="thigh_r")
    FalloffRegionScaleFactors=(PositiveWidth=1,NegativeWidth=0.85,PositiveHeight=1,NegativeHeight=0.3)
    FalloffSize=0.55
    FlipHeightScaling=false
    FlipWidthScaling=true
    OptionalParentItem=(Type=Bone)
    RotationOffset=(X=0,Y=0,Z=0)
   )
   > &RigUnit_ModifyTransforms(
    ItemToModify=((Item=(Type=Bone,Name="dyn_skirt_side_r_01"),Transform=(Rotation=(X=0,Y=0,Z=0,W=1),Translation=(X=0,Y=0,Z=0),Scale3D=(X=1,Y=1,Z=1))))
    ItemToModify.0.Transform.Rotation=&RigVMFunction_MathQuaternionFromEuler(Euler=(X=0,Y=-30,Z=20)).Result
    Weight=@n6.OutputParam
   )
   > &RigUnit_ModifyTransforms(
    ItemToModify=((Item=(Type=Bone,Name="dyn_skirt_side_frnt_r_01"),Transform=(Rotation=(X=0,Y=0,Z=0,W=1),Translation=(X=0,Y=0,Z=-1),Scale3D=(X=1,Y=1,Z=1))))
    ItemToModify.0.Transform.Rotation=&RigVMFunction_MathQuaternionFromEuler(Euler=(X=0,Y=-10,Z=20)).Result
    Weight=@n6.OutputParam
   )
   > &RigUnit_ModifyTransforms(
    ItemToModify=((Item=(Type=Bone,Name="dyn_skirt_side_bk_r_base"),Transform=(Rotation=(X=0,Y=0,Z=0,W=1),Translation=(X=0,Y=0,Z=3),Scale3D=(X=1,Y=1,Z=1))))
    ItemToModify.0.Transform.Rotation=&RigVMFunction_MathQuaternionFromEuler(Euler=(X=0,Y=-5,Z=-20)).Result
    Weight=@n6.OutputParam
   )
   &RigUnit_ParentConstraint(
    AdvancedSettings=(RotationOrderForFilter=XZY)
    Child=(Type=Bone,Name="dyn_skirt_frnt_mid_base")
    Filter=(TranslationFilter=(bX=false,bY=false,bZ=false),RotationFilter=(bX=True,bY=True,bZ=True),ScaleFilter=(bX=True,bY=True,bZ=True))
    Parents=((Item=(Type=Bone,Name="dyn_skirt_side_frnt_l_base"),Weight=0.5),(Item=(Type=Bone,Name="dyn_skirt_side_frnt_r_base"),Weight=0.5))
    Weight=1
    ExecutePin=@SequenceAggregate_1.C
   )
   > &RigUnit_ParentConstraint(
    AdvancedSettings=(RotationOrderForFilter=XZY)
    Child=(Type=Bone,Name="dyn_skirt_frnt_mid_parent")
    Filter=(TranslationFilter=(bX=false,bY=false,bZ=false),RotationFilter=(bX=True,bY=True,bZ=True),ScaleFilter=(bX=True,bY=True,bZ=True))
    Parents=((Item=(Type=Bone,Name="dyn_skirt_side_frnt_l_parent"),Weight=0.5),(Item=(Type=Bone,Name="dyn_skirt_side_frnt_r_parent"),Weight=0.5))
    Weight=1
   )
   > &RigUnit_ParentConstraint(
    AdvancedSettings=(RotationOrderForFilter=XZY)
    Child=(Type=Bone,Name="dyn_skirt_frnt_mid_01")
    Filter=(TranslationFilter=(bX=false,bY=false,bZ=false),RotationFilter=(bX=True,bY=True,bZ=True),ScaleFilter=(bX=True,bY=True,bZ=True))
    Parents=((Item=(Type=Bone,Name="dyn_skirt_side_frnt_l_01"),Weight=0.5),(Item=(Type=Bone,Name="dyn_skirt_side_frnt_r_01"),Weight=0.5))
    Weight=1
   )
   > &RigUnit_ParentConstraint(
    AdvancedSettings=(RotationOrderForFilter=XZY)
    Child=(Type=Bone,Name="dyn_skirt_frnt_mid_02")
    Filter=(TranslationFilter=(bX=false,bY=false,bZ=false),RotationFilter=(bX=True,bY=True,bZ=True),ScaleFilter=(bX=True,bY=True,bZ=True))
    Parents=((Item=(Type=Bone,Name="dyn_skirt_side_frnt_l_02"),Weight=0.5),(Item=(Type=Bone,Name="dyn_skirt_side_frnt_r_02"),Weight=0.5))
    Weight=1
   )
   > &RigUnit_ParentConstraint(
    AdvancedSettings=(RotationOrderForFilter=XZY)
    Child=(Type=Bone,Name="dyn_skirt_frnt_mid_03")
    Filter=(TranslationFilter=(bX=false,bY=false,bZ=false),RotationFilter=(bX=True,bY=True,bZ=True),ScaleFilter=(bX=True,bY=True,bZ=True))
    Parents=((Item=(Type=Bone,Name="dyn_skirt_side_frnt_l_03"),Weight=0.5),(Item=(Type=Bone,Name="dyn_skirt_side_frnt_r_03"),Weight=0.5))
    Weight=1
   )
   > &RigUnit_ModifyTransforms(
    ItemToModify=((Item=(Type=Bone,Name="dyn_skirt_frnt_mid_base"),Transform=(Rotation=(X=0,Y=0,Z=0,W=1),Translation=(X=0,Y=0,Z=0),Scale3D=(X=1,Y=1,Z=1))))
    Weight=&RigVMFunction_MathFloatMul(
     B=8
     A=&RigVMFunction_MathFloatMul(B=@n5.OutputParam, A=@n2.OutputParam) {template="Multiply::Execute(in A,in B,out Result)"}.Result
    ) {template="Multiply::Execute(in A,in B,out Result)"}.Result
    ItemToModify.0.Transform.Rotation=&RigVMFunction_MathQuaternionFromEuler(Euler=(X=0,Y=0,Z=20)).Result
   )
   RigVMFunction_ControlFlowBranch = &RigVMFunction_ControlFlowBranch(ExecuteContext=@SequenceAggregate_1.D, Condition=RigVMVariableNode() {variable=Is Skydive}.Value)
   SequenceAggregate = &RigVMFunction_Sequence(ExecuteContext=@RigVMFunction_ControlFlowBranch.True)
   n7 = &RigUnit_SphericalPoseReader(
    ActiveRegionScaleFactors=(PositiveWidth=3,NegativeWidth=3,PositiveHeight=3,NegativeHeight=3)
    ActiveRegionSize=0.1
    Debug=(bDrawDebug=false,DebugScale=25,DebugSegments=20,DebugThickness=0.25)
    DriverAxis=(X=-1,Y=0,Z=0)
    DriverItem=(Type=Bone,Name="thigh_l")
    FalloffRegionScaleFactors=(PositiveWidth=1,NegativeWidth=1,PositiveHeight=1,NegativeHeight=1)
    FalloffSize=1
    OptionalParentItem=(Type=Bone)
    RotationOffset=(X=-90,Y=0,Z=-80)
    ExecutePin=@SequenceAggregate.A
   )
   > &RigUnit_ModifyTransforms(
    ItemToModify=((Item=(Type=Bone,Name="dyn_skirt_side_bk_l_01"),Transform=(Rotation=(X=0,Y=0,Z=0,W=1),Translation=(X=0,Y=0,Z=0),Scale3D=(X=1,Y=1,Z=1))))
    ItemToModify.0.Transform.Rotation=&RigVMFunction_MathQuaternionFromEuler(Euler=(X=0,Y=0,Z=-15)).Result
    Weight=@n7.OutputParam
   )
   n8 = &RigUnit_GetTransform(Item=(Type=Bone,Name="calf_l"))
   > &RigUnit_ModifyTransforms(
    ItemToModify=((Item=(Type=Bone,Name="dyn_skirt_side_bk_l_base"),Transform=(Rotation=(X=0,Y=0,Z=0,W=1),Translation=(X=0,Y=0,Z=0),Scale3D=(X=1,Y=1,Z=1))))
    Weight=&RigVMFunction_MathDoubleRemap(
     SourceMaximum=-125
     SourceMinimum=-165
     TargetMaximum=1
     TargetMinimum=0
     Value=&RigVMFunction_MathQuaternionToEuler(RotationOrder=ZYX, Value=@n8.Transform.Rotation).Result.Z
    ).Result
    ItemToModify.0.Transform.Rotation=&RigVMFunction_MathQuaternionFromEuler(Euler=(X=0,Y=0,Z=-5)).Result
   )
   n9 = &RigUnit_SphericalPoseReader(
    ActiveRegionScaleFactors=(PositiveWidth=3,NegativeWidth=3,PositiveHeight=3,NegativeHeight=3)
    ActiveRegionSize=0.1
    Debug=(bDrawDebug=false,DebugScale=25,DebugSegments=20,DebugThickness=0.25)
    DriverAxis=(X=1,Y=0,Z=0)
    DriverItem=(Type=Bone,Name="thigh_r")
    FalloffRegionScaleFactors=(PositiveWidth=1,NegativeWidth=1,PositiveHeight=1,NegativeHeight=1)
    FalloffSize=1
    FlipHeightScaling=false
    FlipWidthScaling=true
    OptionalParentItem=(Type=Bone)
    RotationOffset=(X=90,Y=0,Z=-80)
    ExecutePin=@SequenceAggregate.B
   )
   > &RigUnit_ModifyTransforms(
    ItemToModify=((Item=(Type=Bone,Name="dyn_skirt_side_bk_r_01"),Transform=(Rotation=(X=0,Y=0,Z=0,W=1),Translation=(X=0,Y=0,Z=0),Scale3D=(X=1,Y=1,Z=1))))
    ItemToModify.0.Transform.Rotation=&RigVMFunction_MathQuaternionFromEuler(Euler=(X=0,Y=0,Z=-15)).Result
    Weight=@n9.OutputParam
   )
   > &RigUnit_ModifyTransforms(
    ItemToModify=((Item=(Type=Bone,Name="dyn_skirt_bk_mid_01"),Transform=(Rotation=(X=0,Y=0,Z=0,W=1),Translation=(X=0,Y=0,Z=0),Scale3D=(X=1,Y=1,Z=1))))
    Weight=&RigVMFunction_MathFloatMul(B=@n9.OutputParam, A=@n7.OutputParam) {template="Multiply::Execute(in A,in B,out Result)"}.Result
    ItemToModify.0.Transform.Rotation=&RigVMFunction_MathQuaternionFromEuler(Euler=(X=0,Y=0,Z=-20)).Result
   )
   &RigUnit_ParentConstraint(
    AdvancedSettings=(RotationOrderForFilter=XZY)
    Child=(Type=Bone,Name="dyn_skirt_bk_mid_base")
    Filter=(TranslationFilter=(bX=false,bY=false,bZ=false),RotationFilter=(bX=True,bY=True,bZ=True),ScaleFilter=(bX=True,bY=True,bZ=True))
    Parents=((Item=(Type=Bone,Name="dyn_skirt_side_bk_l_base"),Weight=0.5),(Item=(Type=Bone,Name="dyn_skirt_side_bk_r_base"),Weight=0.5))
    Weight=1
    ExecutePin=@SequenceAggregate_1.E
   )
   > &RigUnit_ParentConstraint(
    AdvancedSettings=(RotationOrderForFilter=XZY)
    Child=(Type=Bone,Name="dyn_skirt_bk_mid_parent")
    Filter=(TranslationFilter=(bX=false,bY=false,bZ=false),RotationFilter=(bX=True,bY=True,bZ=True),ScaleFilter=(bX=True,bY=True,bZ=True))
    Parents=((Item=(Type=Bone,Name="dyn_skirt_side_bk_l_parent"),Weight=0.5),(Item=(Type=Bone,Name="dyn_skirt_side_bk_r_parent"),Weight=0.5))
    Weight=1
   )
   > &RigUnit_ParentConstraint(
    AdvancedSettings=(RotationOrderForFilter=XZY)
    Child=(Type=Bone,Name="dyn_skirt_bk_mid_01")
    Filter=(TranslationFilter=(bX=false,bY=false,bZ=false),RotationFilter=(bX=True,bY=True,bZ=True),ScaleFilter=(bX=True,bY=True,bZ=True))
    Parents=((Item=(Type=Bone,Name="dyn_skirt_side_bk_l_01"),Weight=0.5),(Item=(Type=Bone,Name="dyn_skirt_side_bk_r_01"),Weight=0.5))
    Weight=1
   )
   > &RigUnit_ParentConstraint(
    AdvancedSettings=(RotationOrderForFilter=XZY)
    Child=(Type=Bone,Name="dyn_skirt_bk_mid_02")
    Filter=(TranslationFilter=(bX=false,bY=false,bZ=false),RotationFilter=(bX=True,bY=True,bZ=True),ScaleFilter=(bX=True,bY=True,bZ=True))
    Parents=((Item=(Type=Bone,Name="dyn_skirt_side_bk_l_02"),Weight=0.5),(Item=(Type=Bone,Name="dyn_skirt_side_bk_r_02"),Weight=0.5))
    Weight=1
   )
   > n10 = &RigUnit_ParentConstraint(
    AdvancedSettings=(RotationOrderForFilter=XZY)
    Child=(Type=Bone,Name="dyn_skirt_bk_mid_03")
    Filter=(TranslationFilter=(bX=false,bY=false,bZ=false),RotationFilter=(bX=True,bY=True,bZ=True),ScaleFilter=(bX=True,bY=True,bZ=True))
    Parents=((Item=(Type=Bone,Name="dyn_skirt_side_bk_l_03"),Weight=0.5),(Item=(Type=Bone,Name="dyn_skirt_side_bk_r_03"),Weight=0.5))
    Weight=1
   )
   n11 = &RigVMFunction_ControlFlowBranch(ExecuteContext=@n10.ExecutePin, Condition=RigVMVariableNode() {variable=Is Tactical Sprinting}.Value)
   n12 = &RigUnit_GetTransform(Item=(Type=Bone,Name="calf_r"))
   &RigUnit_ModifyTransforms(
    ItemToModify=((Item=(Type=Bone,Name="dyn_skirt_side_bk_r_base"),Transform=(Rotation=(X=0,Y=0,Z=0,W=1),Translation=(X=0,Y=0,Z=0),Scale3D=(X=1,Y=1,Z=1))))
    Weight=&RigVMFunction_MathDoubleRemap(
     SourceMaximum=-125
     SourceMinimum=-165
     TargetMaximum=1
     TargetMinimum=0
     Value=&RigVMFunction_MathQuaternionToEuler(RotationOrder=ZYX, Value=@n12.Transform.Rotation).Result.Z
    ).Result
    ExecutePin=@n11.True
    ItemToModify.0.Transform.Rotation=&RigVMFunction_MathQuaternionFromEuler(Euler=(X=0,Y=0,Z=-20)).Result
   )
   > &RigUnit_ModifyTransforms(
    ItemToModify=((Item=(Type=Bone,Name="dyn_skirt_bk_mid_base"),Transform=(Rotation=(X=0,Y=0,Z=0,W=1),Translation=(X=0,Y=0,Z=0),Scale3D=(X=1,Y=1,Z=1))))
    Weight=1
    ItemToModify.0.Transform.Rotation=&RigVMFunction_MathQuaternionFromEuler(Euler=(X=0,Y=0,Z=-20)).Result
   )
   SequenceAggregate_2 = RigVMAggregateNode(ExecuteContext=@SequenceAggregate_1.F)
   n13 = RigVMVariableNode() {variable=Forward Backward Movement}
   Remap = &RigVMFunction_MathDoubleRemap(SourceMaximum=1, SourceMinimum=0, TargetMaximum=1, TargetMinimum=0, Value=@n13.Value)
   RigVMVariableNode(Value=@Remap.Result, ExecuteContext=@SequenceAggregate_2.A) {variable=Forward Movement}
   > RigVMVariableNode(Value=&RigVMFunction_MathDoubleRemap(SourceMaximum=0, SourceMinimum=-1, TargetMaximum=0, TargetMinimum=1, Value=@n13.Value).Result) {variable=Backward Movement}
   n14 = /Script/RigVM.RigVMFunction_MathVectorAdd(A=(X=0,Y=26,Z=8), B=&RigUnit_GetTransform(Item=(Type=Bone,Name="root"), bInitial=false).Transform.Translation) {is=struct RigVMUnitNode, template="Add::Execute(in A,in B,out Result)"}
   &RigUnit_SetTransform(
    Item=(Type=Control,Name="NewControl")
    Value=(Rotation=(X=0,Y=0,Z=0,W=1),Translation=(X=12188,Y=-7479,Z=3859.7551),Scale3D=(X=1,Y=1,Z=1))
    Weight=1
    bInitial=False
    Value.Translation=@n14.Result
    ExecutePin=@SequenceAggregate_2.D
   )
   n15 = &RigVMFunction_MathBoolToFloat(Value=RigVMVariableNode() {variable=Is Crouching}.Value)
   n16 = &RigUnit_GetRelativeTransformForItem(Child=(Type=Bone,Name="foot_l"), Parent=(Type=Bone,Name="root"), bChildInitial=false, bParentInitial=false)
   n17 = /Script/RigVM.RigVMFunction_MathFloatSub(
    A=1
    B=&RigVMFunction_MathFloatRemap(
     SourceMaximum=0
     SourceMinimum=80
     TargetMinimum=0
     bClamp=true
     Value=&RigVMFunction_MathVectorDistance(A=@n14.Result, B=@n16.RelativeTransform.Translation).Result
    ).Result
   ) {is=struct RigVMUnitNode, template="Subtract::Execute(in A,in B,out Result)"}
   > &RigUnit_ModifyTransforms(
    ItemToModify="((Item=(Type=Bone,Name=\"dyn_skirt_side_frnt_l_base\"),Transform=(Rotation=(X=0,Y=0,Z=0,W=1),Translation=(X=0,Y=0,Z=0),Scale3D=(X=1,Y=1,Z=1))),(Item=(Type=Bone,Name=\"dyn_skirt_frnt_mid_base\"),Transform=(Rotation=(X=0,Y=0,Z=0,W=1),Translation=(X=0,Y=0,Z=0),Sc ...(truncated)"
    Weight=&RigVMFunction_MathFloatMul(
     B=&RigVMFunction_MathFloatMul(A=@n15.Result, B=@n17.Result) {template="Multiply::Execute(in A,in B,out Result)"}.Result
     A=RigVMVariableNode() {variable=Forward Movement}.Value
    ) {template="Multiply::Execute(in A,in B,out Result)"}.Result
    ItemToModify.0.Transform.Rotation=&RigVMFunction_MathQuaternionFromEuler(Euler=(X=0,Y=0,Z=10)).Result
    ItemToModify.1.Transform.Rotation=&RigVMFunction_MathQuaternionFromEuler(Euler=(X=0,Y=0,Z=20)).Result
    ItemToModify.2.Transform.Rotation=&RigVMFunction_MathQuaternionFromEuler(Euler=(X=0,Y=0,Z=0)).Result
   )
   n18 = &RigVMFunction_MathVectorAdd(A=(X=0,Y=0,Z=0), B=&RigUnit_GetTransform(Item=(Type=Bone,Name="root"), bInitial=false).Transform.Translation) {template="Add::Execute(in A,in B,out Result)"}
   > &RigUnit_SetTransform(
    Item=(Type=Control,Name="NewControl")
    Value=(Rotation=(X=0,Y=0,Z=0,W=1),Translation=(X=12188,Y=-7479,Z=3859.7551),Scale3D=(X=1,Y=1,Z=1))
    Weight=1
    bInitial=False
    Value.Translation=@n18.Result
   )
   RigVMFunction_MathBoolNot = &RigVMFunction_MathBoolNot(Value=RigVMVariableNode() {variable=Is Crouching}.Value)
   n19 = &RigVMFunction_MathBoolToFloat(Value=@RigVMFunction_MathBoolNot.Result)
   n20 = &RigUnit_GetRelativeTransformForItem(Child=(Type=Bone,Name="foot_r"), Parent=(Type=Bone,Name="root"), bChildInitial=false, bParentInitial=false)
   n21 = &RigVMFunction_MathFloatSub(
    A=1
    B=&RigVMFunction_MathFloatRemap(
     SourceMaximum=0
     SourceMinimum=60
     TargetMinimum=0
     bClamp=true
     Value=&RigVMFunction_MathVectorDistance(A=@n18.Result, B=@n20.RelativeTransform.Translation).Result
    ).Result
   ) {template="Subtract::Execute(in A,in B,out Result)"}
   > &RigUnit_ModifyTransforms(
    ItemToModify="((Item=(Type=Bone,Name=\"None\"),Transform=(Rotation=(X=0,Y=0,Z=0,W=1),Translation=(X=0,Y=0,Z=0),Scale3D=(X=1,Y=1,Z=1))),(Item=(Type=Bone,Name=\"dyn_skirt_frnt_mid_base\"),Transform=(Rotation=(X=0,Y=0,Z=0,W=1),Translation=(X=0,Y=0,Z=0),Scale3D=(X=1,Y=1,Z=1))), ...(truncated)"
    Weight=&RigVMFunction_MathFloatMul(
     B=&RigVMFunction_MathFloatMul(A=@n19.Result, B=@n21.Result) {template="Multiply::Execute(in A,in B,out Result)"}.Result
     A=RigVMVariableNode() {variable=Forward Movement}.Value
    ) {template="Multiply::Execute(in A,in B,out Result)"}.Result
    ItemToModify.0.Transform.Rotation=&RigVMFunction_MathQuaternionFromEuler(Euler=(X=0,Y=0,Z=0)).Result
    ItemToModify.1.Transform.Rotation=&RigVMFunction_MathQuaternionFromEuler(Euler=(X=0,Y=0,Z=20)).Result
    ItemToModify.2.Transform.Rotation=&RigVMFunction_MathQuaternionFromEuler(Euler=(X=0,Y=0,Z=0)).Result
   )
   n22 = &RigVMFunction_ControlFlowBranch(Condition=RigVMVariableNode() {variable=Is Baseball Sliding}.Value, ExecuteContext=@SequenceAggregate_2.E)
   n23 = &RigUnit_GetTransform(Item=(Type=Bone,Name="calf_r"))
   &RigUnit_ModifyTransforms(
    ItemToModify=((Item=(Type=Bone,Name="dyn_skirt_side_r_base"),Transform=(Rotation=(X=0,Y=0,Z=0,W=1),Translation=(X=0,Y=0,Z=0),Scale3D=(X=1,Y=1,Z=1))))
    Weight=&RigVMFunction_MathDoubleRemap(
     SourceMaximum=-125
     SourceMinimum=-130
     TargetMaximum=1
     TargetMinimum=0
     Value=&RigVMFunction_MathQuaternionToEuler(RotationOrder=ZYX, Value=@n23.Transform.Rotation).Result.Y
    ).Result
    ExecutePin=@n22.True
    ItemToModify.0.Transform.Rotation=&RigVMFunction_MathQuaternionFromEuler(Euler=(X=0,Y=-20,Z=0)).Result
   )
   n24 = RigVMVariableNode() {variable=Left Right Movement}
   RigVMVariableNode(
    Value=&RigVMFunction_MathDoubleRemap(SourceMaximum=0, SourceMinimum=-1, TargetMaximum=0, TargetMinimum=1, Value=@n24.Value).Result
    ExecuteContext=@SequenceAggregate_2.B
   ) {variable=Left Movement}
   > RigVMVariableNode(Value=&RigVMFunction_MathDoubleRemap(SourceMaximum=1, SourceMinimum=0, TargetMaximum=1, TargetMinimum=0, Value=@n24.Value).Result) {variable=Right Movement}
   Add = &RigVMFunction_MathVectorAdd(A=(X=16,Y=18,Z=10), B=&RigUnit_GetTransform(Item=(Type=Bone,Name="root"), bInitial=false).Transform.Translation) {template="Add::Execute(in A,in B,out Result)"}
   RigUnit_SetTransform = &RigUnit_SetTransform(
    Item=(Type=Control,Name="NewControl")
    Value=(Rotation=(X=0,Y=0,Z=0,W=1),Translation=(X=12188,Y=-7479,Z=3859.7551),Scale3D=(X=1,Y=1,Z=1))
    Weight=1
    bInitial=False
    Value.Translation=@Add.Result
    ExecutePin=@SequenceAggregate_2.C
   )
   n25 = &RigVMFunction_MathBoolToFloat(Value=RigVMVariableNode() {variable=Is Crouching}.Value)
   RigVMFunction_MathVectorDistance = &RigVMFunction_MathVectorDistance(
    A=@Add.Result
    B=&RigUnit_GetRelativeTransformForItem(Child=(Type=Bone,Name="foot_l"), Parent=(Type=Bone,Name="root"), bChildInitial=false, bParentInitial=false).RelativeTransform.Translation
   )
   Subtract = &RigVMFunction_MathFloatSub(
    A=1
    B=&RigVMFunction_MathFloatRemap(SourceMaximum=0, SourceMinimum=80, TargetMinimum=0, bClamp=true, Value=@RigVMFunction_MathVectorDistance.Result).Result
   ) {template="Subtract::Execute(in A,in B,out Result)"}
   > &RigUnit_ModifyTransforms(
    ItemToModify="((Item=(Type=Bone,Name=\"dyn_skirt_side_l_base\"),Transform=(Rotation=(X=0,Y=0,Z=0,W=1),Translation=(X=0,Y=0,Z=0),Scale3D=(X=1,Y=1,Z=1))),(Item=(Type=Bone,Name=\"dyn_skirt_frnt_mid_base\"),Transform=(Rotation=(X=0,Y=0,Z=0,W=1),Translation=(X=0,Y=0,Z=0),Scale3D ...(truncated)"
    ItemToModify.0.Transform.Rotation=&RigVMFunction_MathQuaternionFromEuler(Euler=(X=0,Y=-10,Z=15)).Result
    ItemToModify.1.Transform.Rotation=&RigVMFunction_MathQuaternionFromEuler(Euler=(X=0,Y=0,Z=10)).Result
    ItemToModify.2.Transform.Rotation=&RigVMFunction_MathQuaternionFromEuler(Euler=(X=0,Y=0,Z=10)).Result
    Weight=&RigVMFunction_MathFloatMul(
     B=&RigVMFunction_MathFloatMul(A=@n25.Result, B=@Subtract.Result) {template="Multiply::Execute(in A,in B,out Result)"}.Result
     A=RigVMVariableNode() {variable=Left Movement}.Value
    ) {template="Multiply::Execute(in A,in B,out Result)"}.Result
   )
   n26 = &RigVMFunction_MathVectorAdd(A=(X=-5,Y=-45,Z=12), B=&RigUnit_GetTransform(Item=(Type=Bone,Name="root"), bInitial=false).Transform.Translation) {template="Add::Execute(in A,in B,out Result)"}
   > &RigUnit_SetTransform(
    Item=(Type=Control,Name="NewControl")
    Value=(Rotation=(X=0,Y=0,Z=0,W=1),Translation=(X=12188,Y=-7479,Z=3859.7551),Scale3D=(X=1,Y=1,Z=1))
    Weight=1
    bInitial=False
    Value.Translation=@n26.Result
   )
   n27 = &RigVMFunction_MathBoolToFloat(Value=RigVMVariableNode() {variable=Is Crouching}.Value)
   n28 = &RigUnit_GetRelativeTransformForItem(Child=(Type=Bone,Name="foot_r"), Parent=(Type=Bone,Name="root"), bChildInitial=false, bParentInitial=false)
   n29 = &RigVMFunction_MathFloatSub(
    A=1
    B=&RigVMFunction_MathFloatRemap(
     SourceMaximum=0
     SourceMinimum=80
     TargetMinimum=0
     bClamp=true
     Value=&RigVMFunction_MathVectorDistance(A=@n26.Result, B=@n28.RelativeTransform.Translation).Result
    ).Result
   ) {template="Subtract::Execute(in A,in B,out Result)"}
   > &RigUnit_ModifyTransforms(
    ItemToModify="((Item=(Type=Bone,Name=\"dyn_skirt_side_r_base\"),Transform=(Rotation=(X=0,Y=0,Z=0,W=1),Translation=(X=0,Y=0,Z=0),Scale3D=(X=1,Y=1,Z=1))),(Item=(Type=Bone,Name=\"dyn_skirt_frnt_mid_base\"),Transform=(Rotation=(X=0,Y=0,Z=0,W=1),Translation=(X=0,Y=0,Z=0),Scale3D ...(truncated)"
    ItemToModify.0.Transform.Rotation=&RigVMFunction_MathQuaternionFromEuler(Euler=(X=0,Y=-10,Z=0)).Result
    ItemToModify.1.Transform.Rotation=&RigVMFunction_MathQuaternionFromEuler(Euler=(X=0,Y=0,Z=25)).Result
    ItemToModify.2.Transform.Rotation=&RigVMFunction_MathQuaternionFromEuler(Euler=(X=0,Y=-35,Z=10)).Result
    Weight=&RigVMFunction_MathFloatMul(
     B=&RigVMFunction_MathFloatMul(A=@n27.Result, B=@n29.Result) {template="Multiply::Execute(in A,in B,out Result)"}.Result
     A=RigVMVariableNode() {variable=Right Movement}.Value
    ) {template="Multiply::Execute(in A,in B,out Result)"}.Result
   )
  ]}
 SequenceAggregate_1_1_ContainedGraph {nodes [
   n1 = RigVMFunctionEntryNode()
   SequenceAggregate_1 = &RigVMFunction_Sequence(ExecuteContext=@n1.ExecuteContext)
   SequenceAggregate_1_1 = &RigVMFunction_Sequence(ExecuteContext=@SequenceAggregate_1.B)
   SequenceAggregate_1_2 = &RigVMFunction_Sequence(ExecuteContext=@SequenceAggregate_1_1.B)
   SequenceAggregate_1_3 = &RigVMFunction_Sequence(ExecuteContext=@SequenceAggregate_1_2.B)
   SequenceAggregate_1_4 = &RigVMFunction_Sequence(ExecuteContext=@SequenceAggregate_1_3.B)
   SequenceAggregate_1_5 = &RigVMFunction_Sequence(ExecuteContext=@SequenceAggregate_1_4.B)
   RigVMFunctionReturnNode(
    A=@SequenceAggregate_1.A
    B=@SequenceAggregate_1_1.A
    C=@SequenceAggregate_1_2.A
    D=@SequenceAggregate_1_3.A
    E=@SequenceAggregate_1_4.A
    F=@SequenceAggregate_1_5.A
    G=@SequenceAggregate_1_5.B
   )
  ]}
 SequenceAggregate_2_1_ContainedGraph {nodes [
   n1 = RigVMFunctionEntryNode()
   SequenceAggregate_2 = &RigVMFunction_Sequence(ExecuteContext=@n1.ExecuteContext)
   SequenceAggregate_2_1 = &RigVMFunction_Sequence(ExecuteContext=@SequenceAggregate_2.B)
   SequenceAggregate_2_2 = &RigVMFunction_Sequence(ExecuteContext=@SequenceAggregate_2_1.B)
   SequenceAggregate_2_3 = &RigVMFunction_Sequence(ExecuteContext=@SequenceAggregate_2_2.B)
   RigVMFunctionReturnNode(A=@SequenceAggregate_2.A, B=@SequenceAggregate_2_1.A, C=@SequenceAggregate_2_2.A, D=@SequenceAggregate_2_3.A, E=@SequenceAggregate_2_3.B)
  ]}
]
hierarchy [
 root {loc=<0 0 0>, rot=<0 0 0 1>}
 pelvis {parent=root, loc=<0 -0.5301 98.0571>, rot=<0.0127 0.707 -0.0127 -0.707>}
 spine_01 {parent=pelvis, loc=<6.1208 -0.9328 0>, rot=<0 0 0.034 -0.9994>}
 spine_02 {parent=spine_01, loc=<6.5996 -0.2095 0>, rot=<0 0 0.0238 -0.9997>}
 spine_03 {parent=spine_02, loc=<6.4914 -0.1237 0>, rot=<0 0 0.0149 -0.9999>}
 spine_04 {parent=spine_03, loc=<6.9598 0.0794 0>, rot=<0 0 -0.0132 -0.9999>}
 spine_05 {parent=spine_04, loc=<6.9987 0 0>, rot=<0 0 -0.0015 -1>}
 attach_cape {parent=spine_05, loc=<9.3216 -8.4501 0>, rot=<0.0616 -0.7044 -0.0616 -0.7044>}
 neck_01 {parent=spine_05, loc=<11.3138 0.1529 0>, rot=<0 0 -0.2003 -0.9797>}
 neck_02 {parent=neck_01, loc=<4.5627 0 0>, rot=<0 0 0.0504 -0.9987>}
 head {parent=neck_02, loc=<5.3914 0 0>, rot=<0 0 0.115 -0.9934>}
 attach_backpack {parent=spine_05, loc=<0 -8.35 0>, rot=<-0.0093 -0.707 0.0093 -0.707>}
 clavicle_l {parent=spine_05, loc=<5.4831 -0.1053 -3.2946>, rot=<0.0465 -0.7289 -0.009 -0.6829>}
 upperarm_l {parent=clavicle_l, loc=<9.8658 -1.2136 -0.4105>, rot=<0.0164 -0.3245 -0.0269 -0.9454>}
 lowerarm_l {parent=upperarm_l, loc=<26.465 -0.8345 0.259>, rot=<0.0036 0.0161 -0.2298 -0.9731>}
 hand_l {parent=lowerarm_l, loc=<22.7355 -0.3203 -0.6105>, rot=<-0.6784 -0.0223 0.0735 -0.7307>}
 thumb_01_l {parent=hand_l, loc=<1.3118 -0.3191 -2.459>, rot=<0.6546 -0.4042 0.0664 -0.6354>}
 thumb_02_l {parent=thumb_01_l, loc=<5.9336 -0.0448 0.0401>, rot=<-0.0037 0.0213 0.298 -0.9543>}
 thumb_03_l {parent=thumb_02_l, loc=<2.2223 0.3682 -0.0907>, rot=<0.0118 -0.013 0.0066 -0.9998>}
 index_metacarpal_l {parent=hand_l, loc=<1.6535 0.0183 -2.96>, rot=<0.0158 -0.0124 -0.0173 -0.9996>}
 index_01_l {parent=index_metacarpal_l, loc=<8.3008 0.0357 -0.0524>, rot=<-0.0044 0.0485 0.0924 -0.9945>}
 index_02_l {parent=index_01_l, loc=<3.3587 -0.3004 0.0452>, rot=<-0.002 0.0027 0.3225 -0.9466>}
 index_03_l {parent=index_02_l, loc=<2.5578 -0.123 -0.0252>, rot=<-0.002 0.0067 0.2558 -0.9667>}
 middle_metacarpal_l {parent=hand_l, loc=<2.1856 -0.1725 -0.5984>, rot=<0.006 0.0168 -0.0329 -0.9993>}
 middle_01_l {parent=middle_metacarpal_l, loc=<7.3275 -0.2447 0.0129>, rot=<-0.0208 0.0042 0.0858 -0.9961>}
 middle_02_l {parent=middle_01_l, loc=<4.0123 0.1401 0.1175>, rot=<-0.0218 0.0258 0.3512 -0.9357>}
 middle_03_l {parent=middle_02_l, loc=<2.6077 0.0534 0.1373>, rot=<0.0082 0.0188 0.1708 -0.9851>}
 ring_metacarpal_l {parent=hand_l, loc=<2.2074 -0.3625 1.3638>, rot=<-0.0284 0.0722 -0.0458 -0.9959>}
 ring_01_l {parent=ring_metacarpal_l, loc=<7.0295 -0.6197 -0.1446>, rot=<-0.0489 -0.0252 0.1892 -0.9804>}
 ring_02_l {parent=ring_01_l, loc=<3.6174 0.4647 -0.1027>, rot=<0 0 0.2806 -0.9598>}
 ring_03_l {parent=ring_02_l, loc=<2.4152 0.114 0.0199>, rot=<0 0 0.1189 -0.9929>}
 pinky_metacarpal_l {parent=hand_l, loc=<1.7616 -0.9753 3.3814>, rot=<-0.0397 0.0814 -0.069 -0.9935>}
 pinky_01_l {parent=pinky_metacarpal_l, loc=<7.1709 -0.4831 -0.1094>, rot=<-0.0654 -0.0035 0.2242 -0.9723>}
 pinky_02_l {parent=pinky_01_l, loc=<2.8274 0.0893 -0.1043>, rot=<0.0037 0.0011 0.2854 -0.9584>}
 pinky_03_l {parent=pinky_02_l, loc=<2.2046 0.0307 -0.0078>, rot=<-0.0002 0.0098 0.1445 -0.9895>}
 weapon_l {parent=hand_l, loc=<-0.1974 -3.4111 0.3811>, rot=<0 0 0 1>}
 hand_attach_l {parent=hand_l, loc=<8.5 -3 0>, rot=<0.7071 0.7071 0 0>}
 deform_wrist_dn_base_l {parent=hand_l, loc=<0 -1.5326 0>, rot=<0 0 0 1>}
 deform_wrist_dn_l {parent=deform_wrist_dn_base_l, loc=<0 0 0>, rot=<0 0 0 1>}
 deform_wrist_up_base_l {parent=hand_l, loc=<0 2.3009 0>, rot=<0 0 0 1>}
 deform_wrist_up_l {parent=deform_wrist_up_base_l, loc=<0 0 0>, rot=<0 0 0 1>}
 lowerarm_twist_02_l {parent=lowerarm_l, loc=<10.6707 0 0>, rot=<0 0 0 1>}
 lowerarm_twist_01_l {parent=lowerarm_l, loc=<18.316 0 0>, rot=<0 0 0 1>}
 deform_elbow_in_base_l {parent=lowerarm_l, loc=<0.1834 3.6062 -0.2572>, rot=<0 0 0 1>}
 deform_elbow_in_l {parent=deform_elbow_in_base_l, loc=<0 0 0>, rot=<0 0 0 1>}
 deform_elbow_out_base_l {parent=lowerarm_l, loc=<0 -4 0>, rot=<0 0 0 1>}
 deform_elbow_out_l {parent=deform_elbow_out_base_l, loc=<0 0 0>, rot=<0 0 0 1>}
 upperarm_twist_01_l {parent=upperarm_l, loc=<7.2135 -0.1134 0>, rot=<0 0 0 1>}
 deform_upperarm_bk_base_l {parent=upperarm_twist_01_l, loc=<-6.4053 -4.3746 0.314>, rot=<0 0 0 1>}
 deform_upperarm_bk_l {parent=deform_upperarm_bk_base_l, loc=<0 0 0>, rot=<0 0 0 1>}
 deform_upperarm_fr_base_l {parent=upperarm_twist_01_l, loc=<-5.6312 5.092 0.3354>, rot=<0 0 0 1>}
 deform_upperarm_fr_l {parent=deform_upperarm_fr_base_l, loc=<0 0 0>, rot=<0 0 0 1>}
 deform_upperarm_up_base_l {parent=upperarm_twist_01_l, loc=<-8.891 0.4825 3.9517>, rot=<0 0 0 1>}
 deform_upperarm_up_l {parent=deform_upperarm_up_base_l, loc=<0 0 0>, rot=<0 0 0 1>}
 upperarm_twist_02_l {parent=upperarm_l, loc=<19.7998 -0.1134 0>, rot=<0 0 0 1>}
 clavicle_r {parent=spine_05, loc=<5.4831 -0.1053 3.2946>, rot=<0.7289 0.0465 -0.6829 0.009>}
 upperarm_r {parent=clavicle_r, loc=<-9.8658 1.2136 0.4105>, rot=<0.0164 -0.3245 -0.0269 -0.9454>}
 lowerarm_r {parent=upperarm_r, loc=<-26.465 0.8345 -0.259>, rot=<0.0036 0.0161 -0.2298 -0.9731>}
 hand_r {parent=lowerarm_r, loc=<-22.7355 0.3203 0.6106>, rot=<-0.6784 -0.0223 0.0735 -0.7307>}
 thumb_01_r {parent=hand_r, loc=<-1.3118 0.3191 2.459>, rot=<0.6546 -0.4042 0.0664 -0.6354>}
 thumb_02_r {parent=thumb_01_r, loc=<-5.9336 0.0448 -0.0401>, rot=<-0.0037 0.0213 0.298 -0.9543>}
 thumb_03_r {parent=thumb_02_r, loc=<-2.2223 -0.3682 0.0907>, rot=<0.0118 -0.013 0.0066 -0.9998>}
 index_metacarpal_r {parent=hand_r, loc=<-1.6535 -0.0183 2.96>, rot=<0.0158 -0.0124 -0.0173 -0.9996>}
 index_01_r {parent=index_metacarpal_r, loc=<-8.3008 -0.0357 0.0524>, rot=<-0.0044 0.0485 0.0924 -0.9945>}
 index_02_r {parent=index_01_r, loc=<-3.3587 0.3004 -0.0452>, rot=<-0.002 0.0027 0.3225 -0.9466>}
 index_03_r {parent=index_02_r, loc=<-2.5578 0.123 0.0252>, rot=<-0.002 0.0067 0.2558 -0.9667>}
 middle_metacarpal_r {parent=hand_r, loc=<-2.1856 0.1725 0.5984>, rot=<0.006 0.0168 -0.0329 -0.9993>}
 middle_01_r {parent=middle_metacarpal_r, loc=<-7.3275 0.2447 -0.0129>, rot=<-0.0208 0.0042 0.0858 -0.9961>}
 middle_02_r {parent=middle_01_r, loc=<-4.0123 -0.1401 -0.1175>, rot=<-0.0218 0.0258 0.3512 -0.9357>}
 middle_03_r {parent=middle_02_r, loc=<-2.6077 -0.0534 -0.1373>, rot=<0.0082 0.0188 0.1708 -0.9851>}
 ring_metacarpal_r {parent=hand_r, loc=<-2.2074 0.3625 -1.3638>, rot=<-0.0284 0.0722 -0.0458 -0.9959>}
 ring_01_r {parent=ring_metacarpal_r, loc=<-7.0295 0.6197 0.1446>, rot=<-0.0489 -0.0252 0.1892 -0.9804>}
 ring_02_r {parent=ring_01_r, loc=<-3.6174 -0.4647 0.1027>, rot=<0 0 0.2806 -0.9598>}
 ring_03_r {parent=ring_02_r, loc=<-2.4152 -0.114 -0.0199>, rot=<0 0 0.1189 -0.9929>}
 pinky_metacarpal_r {parent=hand_r, loc=<-1.7616 0.9753 -3.3814>, rot=<-0.0397 0.0814 -0.069 -0.9935>}
 pinky_01_r {parent=pinky_metacarpal_r, loc=<-7.1709 0.4831 0.1094>, rot=<-0.0654 -0.0035 0.2242 -0.9723>}
 pinky_02_r {parent=pinky_01_r, loc=<-2.8274 -0.0893 0.1043>, rot=<0.0037 0.0011 0.2854 -0.9584>}
 pinky_03_r {parent=pinky_02_r, loc=<-2.2046 -0.0307 0.0078>, rot=<-0.0002 0.0098 0.1445 -0.9895>}
 weapon_r {parent=hand_r, loc=<0.1974 3.4111 -0.3811>, rot=<0 0 0 1>}
 hand_attach_r {parent=hand_r, loc=<-8.5001 2.9999 0>, rot=<0 0 -0.7071 -0.7071>}
 deform_wrist_dn_base_r {parent=hand_r, loc=<-0.0001 1.5328 0>, rot=<0 0 0 1>}
 deform_wrist_dn_r {parent=deform_wrist_dn_base_r, loc=<0 0 0>, rot=<0 0 0 1>}
 deform_wrist_up_base_r {parent=hand_r, loc=<-0.0002 -2.3007 -0.0001>, rot=<0 0 0 1>}
 deform_wrist_up_r {parent=deform_wrist_up_base_r, loc=<0 0 0>, rot=<0 0 0 1>}
 lowerarm_twist_02_r {parent=lowerarm_r, loc=<-10.6707 0 0>, rot=<0 0 0 -1>}
 lowerarm_twist_01_r {parent=lowerarm_r, loc=<-18.316 0 0>, rot=<0 0 0 -1>}
 deform_elbow_in_base_r {parent=lowerarm_r, loc=<-0.1834 -3.6062 0.2572>, rot=<0 0 0 -1>}
 deform_elbow_in_r {parent=deform_elbow_in_base_r, loc=<0 0 0>, rot=<0 0 0 1>}
 deform_elbow_out_base_r {parent=lowerarm_r, loc=<0 4 -0.0001>, rot=<0 0 0 -1>}
 deform_elbow_out_r {parent=deform_elbow_out_base_r, loc=<0 0 0>, rot=<0 0 0 1>}
 upperarm_twist_01_r {parent=upperarm_r, loc=<-7.2135 0.1134 0>, rot=<0 0 0 -1>}
 deform_upperarm_bk_base_r {parent=upperarm_twist_01_r, loc=<6.4053 4.3746 -0.314>, rot=<0 0 0 -1>}
 deform_upperarm_bk_r {parent=deform_upperarm_bk_base_r, loc=<0 0 0>, rot=<0 0 0 1>}
 deform_upperarm_fr_base_r {parent=upperarm_twist_01_r, loc=<5.6312 -5.092 -0.3354>, rot=<0 0 0 1>}
 deform_upperarm_fr_r {parent=deform_upperarm_fr_base_r, loc=<0 0 0>, rot=<0 0 0 1>}
 deform_upperarm_up_base_r {parent=upperarm_twist_01_r, loc=<8.891 -0.4825 -3.9517>, rot=<0 0 0 1>}
 deform_upperarm_up_r {parent=deform_upperarm_up_base_r, loc=<0 0 0>, rot=<0 0 0 1>}
 upperarm_twist_02_r {parent=upperarm_r, loc=<-19.7998 0.1134 0>, rot=<0 0 0 -1>}
 dyn_jacket_side_frnt_l_base {parent=spine_05, loc=<-3.7958 10.1673 -8.8147>, rot=<0.048 -0.9384 0.3419 -0.0099>}
 dyn_jacket_side_frnt_l_parent {parent=dyn_jacket_side_frnt_l_base, loc=<0 0 0>, rot=<0 0 0 1>}
 dyn_jacket_side_frnt_l_01 {parent=dyn_jacket_side_frnt_l_parent, loc=<8.2596 0 0>, rot=<0 0 0 1>}
 dyn_jacket_side_l_base {parent=spine_05, loc=<-4.4085 2.0537 -11.8467>, rot=<0.0284 -0.7065 0.7065 -0.0284>}
 dyn_jacket_side_l_parent {parent=dyn_jacket_side_l_base, loc=<0 0 0>, rot=<0 0 0 1>}
 dyn_jacket_side_l_01 {parent=dyn_jacket_side_l_parent, loc=<5.7189 0 0>, rot=<0 0 0 1>}
 dyn_jacket_side_bk_l_base {parent=spine_05, loc=<-2.4136 -6.3846 -8.6626>, rot=<-0.346 -0.038 0.0335 -0.9369>}
 dyn_jacket_side_bk_l_parent {parent=dyn_jacket_side_bk_l_base, loc=<0 0 0>, rot=<0 0 0 1>}
 dyn_jacket_side_bk_l_01 {parent=dyn_jacket_side_bk_l_parent, loc=<-6.7424 0 0>, rot=<0 0 0 1>}
 dyn_jacket_side_bk_r_base {parent=spine_05, loc=<-2.4564 -5.7323 9.1015>, rot=<0.0102 0.353 0.9354 -0.0171>}
 dyn_jacket_side_bk_r_parent {parent=dyn_jacket_side_bk_r_base, loc=<0 0 0>, rot=<0 0 0 1>}
 dyn_jacket_side_bk_r_01 {parent=dyn_jacket_side_bk_r_parent, loc=<6.5138 0 0>, rot=<0 0 0 1>}
 dyn_jacket_side_r_base {parent=spine_05, loc=<-4.4088 2.0537 11.8467>, rot=<0.7032 0.0013 -0.0706 -0.7075>}
 dyn_jacket_side_r_parent {parent=dyn_jacket_side_r_base, loc=<0 0 0>, rot=<0 0 0 1>}
 dyn_jacket_side_r_01 {parent=dyn_jacket_side_r_parent, loc=<-5.3539 0 0>, rot=<0 0 0 1>}
 dyn_jacket_side_frnt_r_base {parent=spine_05, loc=<-3.796 10.1673 8.8147>, rot=<0.9338 0.0579 -0.0384 -0.351>}
 dyn_jacket_side_frnt_r_parent {parent=dyn_jacket_side_frnt_r_base, loc=<0 0 0>, rot=<0 0 0 1>}
 dyn_jacket_side_frnt_r_01 {parent=dyn_jacket_side_frnt_r_parent, loc=<-8.1315 0 0>, rot=<0 0 0 1>}
 deform_pec_aim_base_l {parent=spine_04, loc=<3.2456 9.3309 -1.2237>, rot=<0.0603 -0.4033 0.3318 -0.8507>}
 deform_pec_aim_l {parent=deform_pec_aim_base_l, loc=<0 0 0>, rot=<0 0 0 -1>}
 deform_pec_base_l {parent=deform_pec_aim_l, loc=<9.3665 0 0>, rot=<0 0 0 1>}
 deform_pec_l {parent=deform_pec_base_l, loc=<0 0 0>, rot=<0 0 0 1>}
 deform_pec_aim_base_r {parent=spine_04, loc=<3.2452 9.3309 1.2237>, rot=<-0.4033 -0.0603 0.8507 0.3318>}
 deform_pec_aim_r {parent=deform_pec_aim_base_r, loc=<0 0 0>, rot=<0 0 0 -1>}
 deform_pec_base_r {parent=deform_pec_aim_r, loc=<-9.3667 -0.0003 -0.0004>, rot=<0 0 0 1>}
 deform_pec_r {parent=deform_pec_base_r, loc=<0 0 0>, rot=<0 0 0 1>}
 dyn_hip_pouches {parent=spine_02, loc=<2.1266 -7.3177 -0.5334>, rot=<0.0282 -0.7065 -0.0282 -0.7065>}
 thigh_l {parent=pelvis, loc=<-1.0855 -0.2109 -8.7053>, rot=<0.0009 0.0355 0.0155 -0.9993>}
 calf_l {parent=thigh_l, loc=<-42.5284 -0.0019 -0.7297>, rot=<0.0006 -0.0161 -0.0391 -0.9991>}
 foot_l {parent=calf_l, loc=<-46.7401 -2.5014 -0.592>, rot=<0.0427 -0.0238 0.0353 -0.9982>}
 ball_l {parent=foot_l, loc=<-3.5658 10.8066 -0.1544>, rot=<0.0001 -0.0001 -0.7063 -0.7079>}
 calf_twist_02_l {parent=calf_l, loc=<-19.1546 -1.7943 -0.2638>, rot=<0.0237 0.0095 -0.0035 -0.9997>}
 calf_twist_01_l {parent=calf_l, loc=<-35.8483 -1.9197 -0.5794>, rot=<0.0237 0.0095 -0.0035 -0.9997>}
 deform_knee_in_base_l {parent=calf_l, loc=<0.8031 -5.0231 0.2456>, rot=<0 0 0 -1>}
 deform_knee_in_l {parent=deform_knee_in_base_l, loc=<0 0 0>, rot=<0 0 0 1>}
 deform_knee_out_base_l {parent=calf_l, loc=<0 4 0>, rot=<0 0 0 -1>}
 deform_knee_out_l {parent=deform_knee_out_base_l, loc=<0 0 0>, rot=<0 0 0 1>}
 thigh_twist_01_l {parent=thigh_l, loc=<-27.952 0 0>, rot=<0.0013 0 0 -1>}
 deform_groin_base_l {parent=thigh_l, loc=<0.4801 5.6776 -0.2486>, rot=<0 0 0 -1>}
 deform_groin_l {parent=deform_groin_base_l, loc=<0 0 0>, rot=<0 0 0 1>}
 thigh_r {parent=pelvis, loc=<-1.0855 -0.2109 8.7053>, rot=<0.0355 -0.0009 0.9993 0.0155>}
 calf_r {parent=thigh_r, loc=<42.5284 0.0019 0.7297>, rot=<0.0006 -0.0161 -0.0391 -0.9991>}
 foot_r {parent=calf_r, loc=<46.7401 2.5014 0.592>, rot=<0.0427 -0.0238 0.0353 -0.9982>}
 ball_r {parent=foot_r, loc=<3.5658 -10.8066 0.1544>, rot=<0.0001 -0.0001 -0.7063 -0.7079>}
 calf_twist_02_r {parent=calf_r, loc=<19.1546 1.7943 0.2638>, rot=<0.0237 0.0095 -0.0035 -0.9997>}
 calf_twist_01_r {parent=calf_r, loc=<35.8483 1.9197 0.5794>, rot=<0.0237 0.0095 -0.0035 -0.9997>}
 deform_knee_in_base_r {parent=calf_r, loc=<-0.8031 5.0231 -0.2457>, rot=<0 0 0 -1>}
 deform_knee_in_r {parent=deform_knee_in_base_r, loc=<0 0 0>, rot=<0 0 0 1>}
 deform_knee_out_base_r {parent=calf_r, loc=<0 -4 0>, rot=<0 0 0 -1>}
 deform_knee_out_r {parent=deform_knee_out_base_r, loc=<0 0 0>, rot=<0 0 0 1>}
 thigh_twist_01_r {parent=thigh_r, loc=<27.952 0 0>, rot=<0.0013 0 0 -1>}
 deform_groin_base_r {parent=thigh_r, loc=<-0.4801 -5.6776 0.2486>, rot=<0 0 0 -1>}
 deform_groin_r {parent=deform_groin_base_r, loc=<0 0 0>, rot=<0 0 0 1>}
 dyn_skirt_frnt_mid_base {parent=pelvis, loc=<9.996 6.4241 -0.0292>, rot=<-0.0721 -0.9974 0 0>}
 dyn_skirt_frnt_mid_parent {parent=dyn_skirt_frnt_mid_base, loc=<0 0 0>, rot=<0 0 0 1>}
 dyn_skirt_frnt_mid_01 {parent=dyn_skirt_frnt_mid_parent, loc=<12.6971 0 0>, rot=<0 0 0.037 -0.9993>}
 dyn_skirt_frnt_mid_02 {parent=dyn_skirt_frnt_mid_01, loc=<13.1646 0.0001 0>, rot=<0 0 0.0001 -1>}
 dyn_skirt_frnt_mid_03 {parent=dyn_skirt_frnt_mid_02, loc=<13.4644 -0.0275 0>, rot=<0 0 -0.0115 -0.9999>}
 dyn_skirt_side_l_base {parent=pelvis, loc=<10.6704 -0.3021 -9.1316>, rot=<-0.0173 -0.9842 0.0032 -0.1759>}
 dyn_skirt_side_l_parent {parent=dyn_skirt_side_l_base, loc=<0 0 0>, rot=<0 0 0 1>}
 dyn_skirt_side_l_01 {parent=dyn_skirt_side_l_parent, loc=<12.7782 0 0>, rot=<0.0001 -0.0934 0 -0.9956>}
 dyn_skirt_side_l_02 {parent=dyn_skirt_side_l_01, loc=<13.7937 0 0>, rot=<0 -0.0436 0 -0.999>}
 dyn_skirt_side_l_03 {parent=dyn_skirt_side_l_02, loc=<12.9652 0 0.0291>, rot=<0 0 0 -1>}
 dyn_skirt_side_bk_l_base {parent=pelvis, loc=<10.9325 -6.133 -6.6255>, rot=<0.1614 -0.9015 -0.3916 -0.089>}
 dyn_skirt_side_bk_l_parent {parent=dyn_skirt_side_bk_l_base, loc=<0 0 0>, rot=<0 0 0 1>}
 dyn_skirt_side_bk_l_01 {parent=dyn_skirt_side_bk_l_parent, loc=<13.0235 0 0>, rot=<0 0 -0.1496 -0.9887>}
 dyn_skirt_side_bk_l_02 {parent=dyn_skirt_side_bk_l_01, loc=<14.1533 0 0>, rot=<0 0 0.0049 -1>}
 dyn_skirt_side_bk_l_03 {parent=dyn_skirt_side_bk_l_02, loc=<12.4152 0.0064 0>, rot=<0 0 0.0014 -1>}
 dyn_skirt_bk_mid_base {parent=pelvis, loc=<10.8223 -7.1846 -0.0292>, rot=<0.1514 -0.9885 0 0>}
 dyn_skirt_bk_mid_parent {parent=dyn_skirt_bk_mid_base, loc=<0 0 0>, rot=<0 0 0 1>}
 dyn_skirt_bk_mid_01 {parent=dyn_skirt_bk_mid_parent, loc=<13.0235 0 0>, rot=<0 0 -0.1496 -0.9887>}
 dyn_skirt_bk_mid_02 {parent=dyn_skirt_bk_mid_01, loc=<14.1533 0 0>, rot=<0 0 0.0049 -1>}
 dyn_skirt_bk_mid_03 {parent=dyn_skirt_bk_mid_02, loc=<12.4705 0.0149 0>, rot=<0 0 0.0033 -1>}
 dyn_skirt_side_bk_r_base {parent=pelvis, loc=<10.9322 -6.133 6.6255>, rot=<0.9015 0.1614 -0.089 0.3916>}
 dyn_skirt_side_bk_r_parent {parent=dyn_skirt_side_bk_r_base, loc=<0 0 0>, rot=<0 0 0 1>}
 dyn_skirt_side_bk_r_01 {parent=dyn_skirt_side_bk_r_parent, loc=<-13.0233 0.0001 0>, rot=<0 0 -0.1496 -0.9887>}
 dyn_skirt_side_bk_r_02 {parent=dyn_skirt_side_bk_r_01, loc=<-14.1532 0 0>, rot=<0 0 0.0049 -1>}
 dyn_skirt_side_bk_r_03 {parent=dyn_skirt_side_bk_r_02, loc=<-12.4245 -0.0063 0>, rot=<0 0 0.0014 -1>}
 dyn_skirt_side_r_base {parent=pelvis, loc=<10.6699 -0.3021 9.1316>, rot=<0.9842 -0.0173 -0.1759 -0.0032>}
 dyn_skirt_side_r_parent {parent=dyn_skirt_side_r_base, loc=<0 0 0>, rot=<0 0 0 1>}
 dyn_skirt_side_r_01 {parent=dyn_skirt_side_r_parent, loc=<-12.7777 0 -0.0001>, rot=<0.0001 -0.0934 0 -0.9956>}
 dyn_skirt_side_r_02 {parent=dyn_skirt_side_r_01, loc=<-13.7936 0 -0.0001>, rot=<0 -0.0436 0 -0.999>}
 dyn_skirt_side_r_03 {parent=dyn_skirt_side_r_02, loc=<-12.4225 0 -0.0372>, rot=<0 0 0 -1>}
 dyn_skirt_side_frnt_l_base {parent=pelvis, loc=<9.9875 4.522 -6.0119>, rot=<-0.147 -0.9132 0.3744 -0.0647>}
 dyn_skirt_side_frnt_l_parent {parent=dyn_skirt_side_frnt_l_base, loc=<0 0 0>, rot=<0 0 0 1>}
 dyn_skirt_side_frnt_l_01 {parent=dyn_skirt_side_frnt_l_parent, loc=<12.6971 0 0>, rot=<0 0 0.0978 -0.9952>}
 dyn_skirt_side_frnt_l_02 {parent=dyn_skirt_side_frnt_l_01, loc=<13.1646 0.0001 0>, rot=<0 0 0.0001 -1>}
 dyn_skirt_side_frnt_l_03 {parent=dyn_skirt_side_frnt_l_02, loc=<13.4644 -0.0275 0>, rot=<0 0 -0.0115 -0.9999>}
 dyn_skirt_side_frnt_r_base {parent=pelvis, loc=<9.9873 4.522 6.0119>, rot=<0.9132 -0.147 -0.0647 -0.3744>}
 dyn_skirt_side_frnt_r_parent {parent=dyn_skirt_side_frnt_r_base, loc=<0 0 0>, rot=<0 0 0 -1>}
 dyn_skirt_side_frnt_r_01 {parent=dyn_skirt_side_frnt_r_parent, loc=<-12.6969 0 0>, rot=<0 0 0.0978 -0.9952>}
 dyn_skirt_side_frnt_r_02 {parent=dyn_skirt_side_frnt_r_01, loc=<-13.1646 -0.0001 0>, rot=<0 0 0.0001 -1>}
 dyn_skirt_side_frnt_r_03 {parent=dyn_skirt_side_frnt_r_02, loc=<-13.4644 0.0275 0>, rot=<0 0 -0.0115 -0.9999>}
 deform_glute_base_l {parent=pelvis, loc=<-7.3466 -9.0213 -9.1245>, rot=<0.0172 0.031 -0.4477 -0.8935>}
 deform_glute_l {parent=deform_glute_base_l, loc=<0 0 0>, rot=<0 0 0 1>}
 deform_glute_base_r {parent=pelvis, loc=<-7.3464 -9.0213 9.1245>, rot=<0.031 -0.0172 0.8935 -0.4477>}
 deform_glute_r {parent=deform_glute_base_r, loc=<0 0 0>, rot=<0 0 0 1>}
 NewControl {controlType=EULER_TRANSFORM, type=Control, loc=<0 0 0>, rot=<0 0 0 1>}
]
variables [
 Backward Movement {type=double, default=0}
 Forward Backward Movement {type=double, default=0}
 Forward Movement {type=double, default=0}
 Is Baseball Sliding {type=bool, default=False}
 Is Crouching {type=bool, default=False}
 Is Skydive {type=bool, default=False}
 Is Tactical Sprinting {type=bool, default=False}
 Left Movement {type=double, default=0}
 Left Right Movement {type=double, default=0}
 Right Movement {type=double, default=0}
]
