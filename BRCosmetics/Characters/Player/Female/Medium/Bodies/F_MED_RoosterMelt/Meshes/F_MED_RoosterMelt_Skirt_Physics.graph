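# Physics asset for the skirt of the RoosterMelt female medium player body, defining the bodies and joints for its dynamic skirt bones.
# Eight capsule chains ring the skirt: three-bone chains at the back, two-bone chains at the sides and front.
# Each chain starts at a zero-size anchor capsule, with narrow swing limits and near-zero twist between links.
# Unconstrained root and thigh capsules supply body volumes, apparently for skirt collision.
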PhysicsAsset /BRCosmetics/Characters/Player/Female/Medium/Bodies/F_MED_RoosterMelt/Meshes/F_MED_RoosterMelt_Skirt_Physics
bodies {
 dyn_skirt_bk_mid_01 [{shapeType=Capsule, length=0, radius=0, shapeName=dyn_skirt_bk_mid_01_capsule}]
 dyn_skirt_bk_mid_02 [{shapeType=Capsule, length=7.005, radius=4.005, shapeName=dyn_skirt_bk_mid_02_capsule, center=<4.1571 0 0>, rotation=<rot -90 -0.0001 0.0003>}]
 dyn_skirt_bk_mid_03 [{shapeType=Capsule, length=9.6189, radius=5.227, shapeName=dyn_skirt_bk_mid_03_capsule, center=<5.229 0 0>, rotation=<rot 0 -179.9997 -180>}]
 dyn_skirt_frnt_mid_02 [{shapeType=Capsule, length=0, radius=0, shapeName=dyn_skirt_frnt_mid_02_capsule, rotation=<rot 0 0.0002 0>}]
 dyn_skirt_frnt_mid_03 [{shapeType=Capsule, length=7, radius=4, shapeName=dyn_skirt_frnt_mid_03_capsule, center=<3.8591 0 0>, rotation=<rot 90 -179.9899 -179.9905>}]
 dyn_skirt_side_bk_l_01 [{shapeType=Capsule, length=0, radius=0, shapeName=dyn_skirt_side_bk_l_01_capsule}]
 dyn_skirt_side_bk_l_02 [{
   shapeType=Capsule
   length=7.1519
   radius=3.9229
   shapeName=dyn_skirt_side_bk_l_02_capsule
   center=<4.648 0 0>
   rotation=<rot -90 90 -89.9998>
  }]
 dyn_skirt_side_bk_l_03 [{
   shapeType=Capsule
   length=9.3402
   radius=4.9007
   shapeName=dyn_skirt_side_bk_l_03_capsule
   center=<4.4572 0 0>
   rotation=<rot -0.0002 -179.9996 -150>
  }]
 dyn_skirt_side_bk_r_01 [{shapeType=Capsule, length=0, radius=0, shapeName=dyn_skirt_side_bk_r_01_capsule}]
 dyn_skirt_side_bk_r_02 [{
   shapeType=Capsule
   length=8.005
   radius=4.005
   shapeName=dyn_skirt_side_bk_r_02_capsule
   center=<-5.0589 0 0>
   rotation=<rot 85 0.0002 -0.0001>
  }]
 dyn_skirt_side_bk_r_03 [{
   shapeType=Capsule
   length=8.4123
   radius=4.8528
   shapeName=dyn_skirt_side_bk_r_03_capsule
   center=<-5.2148 0 0>
   rotation=<rot -0.3185 -179.9448 -179.9933>
  }]
 dyn_skirt_side_frnt_l_02 [{shapeType=Capsule, length=0, radius=0, shapeName=dyn_skirt_side_frnt_l_02_capsule, rotation=<rot -0.0002 0.0002 0>}]
 dyn_skirt_side_frnt_l_03 [{
   shapeType=Capsule
   length=7.5
   radius=3.5
   shapeName=dyn_skirt_side_frnt_l_03_capsule
   center=<3.9567 0 0>
   rotation=<rot 90 -160.5289 -160.5293>
  }]
 dyn_skirt_side_frnt_r_02 [{shapeType=Capsule, length=0, radius=0, shapeName=dyn_skirt_side_frnt_r_02_capsule, rotation=<rot -0.0001 0.0002 0>}]
 dyn_skirt_side_frnt_r_03 [{shapeType=Capsule, length=7.5, radius=3.5, shapeName=dyn_skirt_side_frnt_r_03_capsule, center=<-3.8895 0 0>, rotation=<rot 90 180 179.9995>}]
 dyn_skirt_side_l_02 [{shapeType=Capsule, length=0, radius=0, shapeName=dyn_skirt_side_l_02_capsule}]
 dyn_skirt_side_l_03 [{shapeType=Capsule, length=10.6451, radius=3.3901, shapeName=dyn_skirt_side_l_03_capsule, center=<7.4997 0 0>, rotation=<rot 90 -90 -90.0002>}]
 dyn_skirt_side_r_02 [{shapeType=Capsule, length=0, radius=0, shapeName=dyn_skirt_side_r_02_capsule}]
 dyn_skirt_side_r_03 [{
   shapeType=Capsule
   length=10.7303
   radius=3.6911
   shapeName=dyn_skirt_side_r_03_capsule
   center=<-8.0089 0 0>
   rotation=<rot -90 179.2404 -179.2403>
  }]
 root [
  {shapeType=Capsule, length=39, radius=21.505, shapeName=root_capsule, center=<-5.4795 -8.5965 12.7692>, rotation=<rot 0.0009 65 -90.0004>}
  {shapeType=Capsule, length=27.5, radius=17.5, shapeName=root_capsule, center=<-24.9001 -18.0896 13.483>, rotation=<rot 0.0003 155 90.0005>}
 ]
 thigh_l [{
   shapeType=Capsule
   length=36.1359
   radius=5.6495
   shapeName=thigh_l_capsule
   center=<-28.4528 0.276 0.4176>
   rotation=<rot 0 86.6024 -87.8812>
  }]
 thigh_r [{
   shapeType=Capsule
   length=35.6168
   radius=5.638
   shapeName=thigh_r_capsule
   center=<28.4916 0.0244 -0.5565>
   rotation=<rot 0 87.7845 -88.0566>
  }]
}
constraints [
 {bone1Name=dyn_skirt_bk_mid_02, bone2Name=dyn_skirt_bk_mid_01, swing1LimitDegrees=30, swing2LimitDegrees=10, twistLimitDegrees=1}
 {bone1Name=dyn_skirt_bk_mid_03, bone2Name=dyn_skirt_bk_mid_02, swing1LimitDegrees=10, swing2LimitDegrees=20, twistLimitDegrees=1}
 {bone1Name=dyn_skirt_side_bk_l_03, bone2Name=dyn_skirt_side_bk_l_02, swing1LimitDegrees=10, swing2LimitDegrees=20, twistLimitDegrees=1}
 {bone1Name=dyn_skirt_side_bk_l_02, bone2Name=dyn_skirt_side_bk_l_01, swing1LimitDegrees=10, swing2LimitDegrees=10, twistLimitDegrees=1}
 {bone1Name=dyn_skirt_side_bk_r_02, bone2Name=dyn_skirt_side_bk_r_01, swing1LimitDegrees=10, swing2LimitDegrees=10, twistLimitDegrees=1}
 {bone1Name=dyn_skirt_side_bk_r_03, bone2Name=dyn_skirt_side_bk_r_02, swing1LimitDegrees=10, swing2LimitDegrees=20, twistLimitDegrees=1}
 {bone1Name=dyn_skirt_side_l_03, bone2Name=dyn_skirt_side_l_02, swing1LimitDegrees=20, swing2LimitDegrees=10, twistLimitDegrees=0}
 {bone1Name=dyn_skirt_side_r_03, bone2Name=dyn_skirt_side_r_02, swing1LimitDegrees=20, swing2LimitDegrees=10, twistLimitDegrees=1}
 {bone1Name=dyn_skirt_frnt_mid_03, bone2Name=dyn_skirt_frnt_mid_02, swing1LimitDegrees=10, swing2LimitDegrees=20, twistLimitDegrees=1}
 {bone1Name=dyn_skirt_side_frnt_l_03, bone2Name=dyn_skirt_side_frnt_l_02, swing1LimitDegrees=5, swing2LimitDegrees=20, twistLimitDegrees=1}
 {bone1Name=dyn_skirt_side_frnt_r_03, bone2Name=dyn_skirt_side_frnt_r_02, swing1LimitDegrees=5, swing2LimitDegrees=20, twistLimitDegrees=1}
]
