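# Character part AnimBlueprint for the RoosterMelt female body that adds secondary motion to its skirt and jacket.
# The copied parent pose passes through skirt Control Rig, skirt physics, DBNO offsets, jacket physics, pouch constraint and deform rig.
# Bone tweaks then shrink correctives when rigid bodies are off and rotate skirt bones while falling.
# The EventGraph updates movement flags, gravity override and material WPO cloth wind, and detects backblings on customization load.
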AnimBlueprint /BRCosmetics/Characters/Player/Female/Medium/Bodies/F_MED_RoosterMelt/Meshes/F_MED_RoosterMelt_AnimBP
target_skeleton=&/F_MED_RoosterMelt_Skeleton
graphs [
 AnimGraph {nodes [
   n1 = Get Is Falling() -> (Is Falling:double)
   n2 = "Transform (Modify) Bone - Bone: dyn_skirt_side_bk_r_01"(
    ComponentPose=Local To Component(LocalPose=Use cached pose 'ClippingCache'())
    Alpha=@n1
    Translation="0,0,0"
    Rotation="0,-25,0"
    Scale="1,1,1"
   ) {
    rotationMode=BMM_Additive
    rotationSpace=BCS_BoneSpace
    alphaBoolBlend {blendOption=Linear}
    alphaScaleBiasClamp {bInterpResult=true, interpSpeedDecreasing=6, interpSpeedIncreasing=3}
    boneToModify {boneName=dyn_skirt_side_bk_r_01}
    rotation {yaw=-25}
   }
   n3 = "Transform (Modify) Bone - Bone: dyn_skirt_bk_mid_01"(ComponentPose=@n2, Alpha=@n1, Translation="0,0,0", Rotation="0,-30,0", Scale="1,1,1") {
    rotationMode=BMM_Additive
    rotationSpace=BCS_BoneSpace
    alphaBoolBlend {blendOption=Linear}
    alphaScaleBiasClamp {bInterpResult=true, interpSpeedDecreasing=6, interpSpeedIncreasing=3}
    boneToModify {boneName=dyn_skirt_bk_mid_01}
    rotation {yaw=-30}
   }
   n4 = "Transform (Modify) Bone - Bone: dyn_skirt_side_bk_l_01"(ComponentPose=@n3, Alpha=@n1, Translation="0,0,0", Rotation="0,-25,0", Scale="1,1,1") {
    rotationMode=BMM_Additive
    rotationSpace=BCS_BoneSpace
    alphaBoolBlend {blendInTime=0.6, blendOption=QuadraticInOut, blendOutTime=0.4}
    alphaScaleBiasClamp {bInterpResult=true, interpSpeedDecreasing=6, interpSpeedIncreasing=3}
    boneToModify {boneName=dyn_skirt_side_bk_l_01}
    rotation {yaw=-25}
   }
   n5 = "Transform (Modify) Bone - Bone: dyn_skirt_side_l_03"(ComponentPose=@n4, Alpha=@n1, Translation="0,0,0", Rotation="25,0,0", Scale="1,1,1") {
    rotationMode=BMM_Additive
    rotationSpace=BCS_BoneSpace
    alphaBoolBlend {blendInTime=0.6, blendOption=QuadraticInOut, blendOutTime=0.4}
    alphaScaleBiasClamp {bInterpResult=true, interpSpeedDecreasing=6, interpSpeedIncreasing=3}
    boneToModify {boneName=dyn_skirt_side_l_03}
    rotation {pitch=25}
   }
   n6 = "Transform (Modify) Bone - Bone: dyn_skirt_side_r_03"(ComponentPose=@n5, Alpha=@n1, Translation="0,0,0", Rotation="-25,0,0", Scale="1,1,1") {
    rotationMode=BMM_Additive
    rotationSpace=BCS_BoneSpace
    alphaBoolBlend {blendInTime=0.6, blendOption=QuadraticInOut, blendOutTime=0.4}
    alphaScaleBiasClamp {bInterpResult=true, interpSpeedDecreasing=6, interpSpeedIncreasing=3}
    boneToModify {boneName=dyn_skirt_side_r_03}
    rotation {pitch=-25}
   }
   Output Pose(Result=Component To Local(ComponentPose=@n6)) {layerGroup=None}
   n7 = RigidBody(ComponentPose=Local To Component(LocalPose=Use cached pose 'DBNOCache'())) {
    bClampLinearTranslationLimitToRefPose=true
    bOverrideWorldGravity=true
    overridePhysicsAsset=&/F_MED_RoosterMelt_Jacket_Physics
    simulationSpace=BaseBoneSpace
    alphaBoolBlend {blendOption=Linear}
    baseBoneRef {boneName=spine_05}
    componentAppliedLinearAccClamp=<5000 5000 5000>
    componentLinearAccScale=<0.3 0.3 0.3>
    componentLinearVelScale=<0.3 0.3 0.3>
    overrideWorldGravity {z=-980}
    simSpaceSettings {maxAngularAcceleration=500, maxAngularVelocity=3, maxLinearAcceleration=400, maxLinearVelocity=250, velocityScaleZ=0.12, worldAlpha=1}
   }
   n8 = "Constraint - dyn_hip_pouches (PP)"!(ComponentPose=@n7, Alpha=1, ConstraintWeights_0=0.9999, ConstraintWeights_1:float=0.0001) -> (Pose) {
    alphaBoolBlend {blendOption=Linear}
    boneToModify {boneName=dyn_hip_pouches}
    constraintSetup [{targetBone {boneName=spine_02}}, {targetBone {boneName=spine_03}}]
    constraintWeights [1,1]
   }
   n9 = Control Rig(Source=Component To Local(ComponentPose=@n8)) {
    alphaBoolBlend {blendOption=Linear}
    controlRigAssetReference {blueprintRigClass=&/F_MED_RoosterMelt_DeformRig.F_MED_RoosterMelt_DeformRig_C}
   }
   Save cached pose 'DeformRigCache'(Pose=@n9) {cacheName=DeformRigCache}
   n10 = Control Rig!(
    Source=Copy Pose From Mesh() {bUseAttachedParent=true}
    Alpha
    Is Skydive:bool=Get Full SkyDive() -> (Full SkyDive:bool)
    Left Right Movement:double=Get Left Right Movement() -> (Left Right Movement:double)
    Forward Backward Movement:double=Get Forward Backward Movement() -> (Forward Backward Movement:double)
    Is Crouching:bool=Get Is Clean Crouching() -> (Is Clean Crouching:bool)
    Is Baseball Sliding:bool=Get bIsBaseballSliding()
    Is Tactical Sprinting:bool=Get bIsTacticalSprinting()
   ) -> (Pose) {
    alphaBoolBlend {blendOption=Linear}
    controlRigAssetReference {blueprintRigClass=&/F_MED_RoosterMelt_Skirt_CtrlRig.F_MED_RoosterMelt_Skirt_CtrlRig_C}
   }
   n11 = RigidBody(ComponentPose=Local To Component(LocalPose=@n10), ExternalForce=Get Gravity Override()) {
    bClampLinearTranslationLimitToRefPose=true
    bOverrideWorldGravity=true
    overridePhysicsAsset=&/F_MED_RoosterMelt_Skirt_Physics
    simulationSpace=BaseBoneSpace
    alphaBoolBlend {blendOption=Linear}
    baseBoneRef {boneName=pelvis}
    componentAppliedLinearAccClamp=<3000 3000 3000>
    componentLinearAccScale=<1.4 1.3 1.3>
    componentLinearVelScale=<1.4 1.3 1.3>
    overrideWorldGravity {z=-980}
    simSpaceSettings {maxAngularAcceleration=400, maxAngularVelocity=3, maxLinearAcceleration=500, maxLinearVelocity=250, velocityScaleZ=0, worldAlpha=1}
   }
   Save cached pose 'SSCCache'(Pose=Component To Local(ComponentPose=@n11)) {cacheName=SSCCache}
   n12 = Get bIsDBNO()
   n13 = "Transform (Modify) Bone - Bone: dyn_skirt_side_bk_l_base"!(
    ComponentPose=Local To Component(LocalPose=Use cached pose 'SSCCache'())
    bAlphaBoolEnabled:bool=@n12
    Translation="0,0,0"
    Rotation="-10,-7.5,-2"
    Scale="1,1,1"
   ) -> (Pose) {
    alphaInputType=Bool
    rotationMode=BMM_Additive
    rotationSpace=BCS_BoneSpace
    translationMode=BMM_Additive
    translationSpace=BCS_BoneSpace
    alphaBoolBlend {blendOption=Linear}
    alphaScaleBiasClamp {bInterpResult=true}
    boneToModify {boneName=dyn_skirt_side_bk_l_base}
    rotation=<rot -10 -7.5 -2>
   }
   n14 = "Transform (Modify) Bone - Bone: dyn_skirt_side_bk_l_01"!(ComponentPose=@n13, bAlphaBoolEnabled:bool=@n12, Translation="0,0,0", Rotation="-10,0,0", Scale="1,1,1") -> (Pose) {
    alphaInputType=Bool
    rotationMode=BMM_Additive
    rotationSpace=BCS_BoneSpace
    translationMode=BMM_Additive
    translationSpace=BCS_BoneSpace
    alphaBoolBlend {blendOption=Linear}
    alphaScaleBiasClamp {bInterpResult=true}
    boneToModify {boneName=dyn_skirt_side_bk_l_01}
    rotation {pitch=-10}
   }
   n15 = "Transform (Modify) Bone - Bone: dyn_skirt_side_bk_r_base"!(ComponentPose=@n14, bAlphaBoolEnabled:bool=@n12, Translation="0,0,0", Rotation="-10,-10,0", Scale="1,1,1") -> (Pose) {
    alphaInputType=Bool
    rotationMode=BMM_Additive
    rotationSpace=BCS_BoneSpace
    translationMode=BMM_Additive
    translationSpace=BCS_BoneSpace
    alphaBoolBlend {blendOption=Linear}
    alphaScaleBiasClamp {bInterpResult=true}
    boneToModify {boneName=dyn_skirt_side_bk_r_base}
    rotation {pitch=-10, yaw=-10}
   }
   n16 = "Transform (Modify) Bone - Bone: dyn_skirt_bk_mid_base"!(ComponentPose=@n15, bAlphaBoolEnabled:bool=@n12, Translation="0,0,0", Rotation="-10,0,-10", Scale="1,1,1") -> (Pose) {
    alphaInputType=Bool
    rotationMode=BMM_Additive
    rotationSpace=BCS_BoneSpace
    translationMode=BMM_Additive
    translationSpace=BCS_BoneSpace
    alphaBoolBlend {blendOption=Linear}
    alphaScaleBiasClamp {bInterpResult=true}
    boneToModify {boneName=dyn_skirt_bk_mid_base}
    rotation {pitch=-10, roll=-10}
   }
   Save cached pose 'DBNOCache'(Pose=Component To Local(ComponentPose=@n16)) {cacheName=DBNOCache}
   n17 = Get IsRigidBodyOff() -> (IsRigidBodyOff:bool)
   n18 = "Transform (Modify) Bone - Bone: dyn_spine_shirt_scale"!(
    ComponentPose=Local To Component(LocalPose=Use cached pose 'DeformRigCache'())
    bAlphaBoolEnabled:bool=@n17
    Translation="0,0,0"
    Rotation="0,0,0"
    Scale="-0.7,-0.7,-0.7"
   ) -> (Pose) {
    alphaInputType=Bool
    rotationSpace=BCS_BoneSpace
    scaleMode=BMM_Additive
    scaleSpace=BCS_BoneSpace
    translationSpace=BCS_BoneSpace
    alphaBoolBlend {blendOption=Linear}
    boneToModify {boneName=dyn_spine_shirt_scale}
    scale=<-0.7 -0.7 -0.7>
   }
   n19 = "Transform (Modify) Bone - Bone: dyn_neck_correct"!(ComponentPose=@n18, bAlphaBoolEnabled:bool=@n17, Translation="0,0,0", Rotation="0,0,0", Scale="-0.3,1,-0.3") -> (Pose) {
    alphaInputType=Bool
    rotationSpace=BCS_BoneSpace
    scaleMode=BMM_Additive
    scaleSpace=BCS_BoneSpace
    translationSpace=BCS_BoneSpace
    alphaBoolBlend {blendOption=Linear}
    boneToModify {boneName=dyn_neck_correct}
    scale {x=-0.3, z=-0.3}
   }
   Save cached pose 'ClippingCache'(Pose=Component To Local(ComponentPose=@n19)) {cacheName=ClippingCache}
  ]}
 EventGraph {nodes [
   n1 = Event BlueprintUpdateAnimation()
   > n2 = Sequence!(execute) -> (then_0, then_1, then_2:Exec, then_3:Exec)
   > Set Full SkyDive(Full SkyDive:bool=OR Boolean!(A=Get bIsSkydiveFloating(), B=Get bIsSkydiveDiveMode(), C:bool=Get bIsSkyDiving(), D:bool, E:bool) -> (ReturnValue)) -> (Output_Get:bool)
   n3 = Get bIsSlopeSliding() -> (bIsSlopeSliding:bool)
   n4 = Get bIsSliding()
   n5 = Get bIsKneeSliding()
   n6 = Get bIsBaseballSliding()
   > Set Is Clean Crouching(Is Clean Crouching:bool=AND Boolean!(A=Get bIsCrouching(), B=NOT Boolean(A=@n5), C:bool=NOT Boolean(A=@n4), D:bool=NOT Boolean(A=@n6), E:bool=NOT Boolean(A=@n3)) -> (ReturnValue)) -> (Output_Get:bool)
   > Set Is LOD2(Is LOD2:bool="integer > integer"(A=Get CurrentLOD(), B=1)) -> (Output_Get:bool)
   > Set Is Falling(Is Falling:double="To Float (Boolean)"(InBool=Get bIsFalling())) -> (Output_Get:double)
   > Set IsRigidBodyOff(IsRigidBodyOff:bool=NOT Boolean(A=Get bRigidBodyEnabled())) -> (Output_Get:bool)
   n7 = Gravity Override State Function DEPRECATED(
    execute=@n2.then_1
    CharacterPartAnimInstance=Self-Reference()
    GameTimeInSeconds=GetGameTimeInSeconds()
    GravityOverrideParams=Get Gravity Override Params_0() -> (Gravity Override Params_0:Gravity Override Params Struct Structure)
   )
   > Set Gravity Override(Gravity Override=@n7)
   n8 = Get Mesh!() -> (Mesh)
   n9 = Get Overlay Mat Instance Dyanmic() -> (Overlay Mat Instance Dyanmic:Material Instance Dynamic Object Reference)
   n10 = Get Material Instance Dynamic() -> (Material Instance Dynamic:Material Instance Dynamic Object Reference)
   > Set WPO Globals Function(
    Material=@n10
    Overlay Material=@n9
    Skeletal Mesh=@n8
    "Local Noise Transform  Location"="Get Local Noise Transform  Location"() -> ("Local Noise Transform  Location":Vector)
    Local Noise Offset=Get Local Noise Offset()
   )
   > Set WPO Regional Function!(
    execute
    Material:Material Instance Dynamic Object Reference=@n10
    Overlay Material:Material Instance Dynamic Object Reference=@n9
    Skeletal Mesh:Skeletal Mesh Component Object Reference=@n8
    Velocity Transform:Name=pelvis
    Velocity Interp Speed:double=Get Velocity Interp Speed()
    Collision Transform:Name=pelvis
    Collision Transform Axis:EAxis Enum
    Invert Collision Transform Axis:bool=false
    Delta Time X:double=@n1.DeltaTimeX
    Previous Velocity Transform Location:"Vector (by ref)"=Get R2 Velocity Transform Location()
    Previous Velocity Transform Velocity:"Vector (by ref)"=Get R2 Velocity Transform Velocity()
    Region:int=1
   ) -> (then)
   > Set WPO Regional Function(
    Material=@n10
    Overlay Material=@n9
    Skeletal Mesh=@n8
    Velocity Transform=spine_2
    Velocity Interp Speed=Get Velocity Interp Speed()
    Collision Transform=spine_2
    Invert Collision Transform Axis=true
    Delta Time X=@n1.DeltaTimeX
    Previous Velocity Transform Location=Get L Previous Velocity Transform Location() -> (L Previous Velocity Transform Location:Vector)
    Previous Velocity Transform Velocity=Get L Previous Velocity Transform Velocity() -> (L Previous Velocity Transform Velocity:Vector)
    Region=2
   )
   > SetScalarParameterValue(self=[@n10, @n9], ParameterName=ClothSim_WindStrength, Value=Get WPO Wind() -> (WPO Wind:double))
   > SetScalarParameterValue(self=[@n10, @n9], ParameterName=ClothSim_TurbulenceNoiseTileDivisor, Value=Get Turbulence Noise() -> (Turbulence Noise:float))
   n11 = PlayerMovement(execute=@n2.then_2, AnimInstance=Self-Reference(), XDotProductCoefficient=1, YDotProductCoefficient=1, ZDotProductCoefficient=1)
   > Set Left Right Movement(Left Right Movement:double=@n11) -> (Output_Get:double)
   > Set Forward Backward Movement(Forward Backward Movement:double=@n11.YMovement) -> (Output_Get:double)
   n12 = Branch(
    execute=@n2.then_3
    Condition=OR Boolean!(A=Get bIsSkydiveFloating(), B=Get bIsSkydiveDiveMode(), C=Get bIsParachuteOpened()) -> (ReturnValue)
   )
   > Set Velocity Interp Speed(Velocity Interp Speed:double=10) -> (Output_Get:double)
   > Set WPO Wind(WPO Wind:double=1) -> (Output_Get:double)
   > Set Turbulence Noise(Turbulence Noise:float=200) -> (Output_Get:float)
   Set Velocity Interp Speed(execute=@n12.else, Velocity Interp Speed=0)
   > Set WPO Wind(WPO Wind=0)
   > Set Turbulence Noise(Turbulence Noise=0)
   Event OnCharacterCustomizationLoaded()
   > n13 = Sequence!(execute) -> (then_0, then_1)
   n14 = CheckForBackbling(LongBackblingThreshold=40)
   > Branch(Condition=@n14)
   > n15 = Sequence!(execute) -> (then_0, then_1, then_2:Exec)
   > Set IsBackblingOn(IsBackblingOn:bool=@n14) -> (Output_Get:bool)
   n16 = Branch(execute=@n15.then_1, Condition=OR Boolean!(A=@n14.bIsBackblingACape, B=@n14.bIsBackblingLong) -> (ReturnValue))
   > Set IsLongBackblingOn(IsLongBackblingOn:bool=true) -> (Output_Get:bool)
   > Set IsShortBackblingOn(IsShortBackblingOn:bool=false) -> (Output_Get:bool)
   Set IsLongBackblingOn(execute=@n16.else, IsLongBackblingOn=false)
   > Set IsShortBackblingOn(IsShortBackblingOn=true)
   n17 = GetPartSkeletalMeshComponent()
   n18 = Set Mesh(execute=@n13.then_1, Mesh:Skeletal Mesh Component Object Reference=@n17) -> (Output_Get:Skeletal Mesh Component Object Reference)
   > n19 = Get Dynamic Material Instance(SkeletalMeshComponent=@n18)
   > Set Material Instance Dynamic(Material Instance Dynamic:Material Instance Dynamic Object Reference=@n19) -> (Output_Get:Material Instance Dynamic Object Reference)
   > n20 = GetOrCreateDynamicOverlayMaterial(MeshComponent=@n17, MaterialName=None)
   > Set Overlay Mat Instance Dyanmic(Overlay Mat Instance Dyanmic:Material Instance Dynamic Object Reference=@n20) -> (Output_Get:Material Instance Dynamic Object Reference)
  ]}
]
variables [
 Forward Backward Movement {type=number}
 Full SkyDive {type=boolean}
 Gravity Override {type=Vector}
 Gravity Override Params {
  type=GravityOverrideParamsStruct
  default {
   skydive_DivingDown {jointName=root}
   skydive_FloatingDown {jointName=pelvis, gravityOverride=<0 0 0>}
   skydive_GoingUp {jointName=pelvis, gravityOverride=<0 0 0>}
  }
 }
 Gravity Override Params_0 {type=GravityOverrideParamsStruct, default {rail_Grinding {gravityOverride=<0 0 0>}}}
 Is Clean Crouching {type=boolean}
 Is Falling {type=number}
 Is LOD2 {type=boolean}
 IsBackblingOn {type=boolean}
 IsLongBackblingOn {type=boolean}
 IsRigidBodyOff {type=boolean}
 IsShortBackblingOn {type=boolean}
 L Previous Velocity Transform Location {type=Vector}
 L Previous Velocity Transform Velocity {type=Vector}
 Left Right Movement {type=number}
 Local Noise Offset {type=Vector}
 "Local Noise Transform  Location" {type=Vector}
 Material Instance Dynamic {type=MaterialInstanceDynamic}
 Mesh {type=SkeletalMeshComponent}
 Overlay Mat Instance Dyanmic {type=MaterialInstanceDynamic}
 R2 Velocity Transform Location {type=Vector}
 R2 Velocity Transform Velocity {type=Vector}
 Turbulence Noise {type=number}
 Velocity Interp Speed {type=number}
 WPO Wind {type=number}
]
