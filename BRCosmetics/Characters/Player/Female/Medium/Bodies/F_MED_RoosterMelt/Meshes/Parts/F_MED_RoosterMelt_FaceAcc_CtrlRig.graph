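# Control Rig for the F_MED_RoosterMelt face accessory part that offsets hair bones for a rightward head tilt.
# The forward solve chains three Modify Transforms nodes on the left, back-middle and right hair base bones, all weighted by the Head Tilt Right variable.
# The back base rotates, the right base rotates and shifts, and the left base's offset is identity.
# Only these base bones are edited, above their dyn_hair child chains.
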ControlRigBlueprint /BRCosmetics/Characters/Player/Female/Medium/Bodies/F_MED_RoosterMelt/Meshes/Parts/F_MED_RoosterMelt_FaceAcc_CtrlRig
graphs [
 RigVMFunctionLibrary {links [], nodes []}
 RigVMModel {nodes [
   RigUnit_BeginExecution = &RigUnit_BeginExecution()
   SequenceAggregate_1 = &RigVMFunction_Sequence(ExecuteContext=@RigUnit_BeginExecution.ExecutePin)
   n1 = RigVMVariableNode() {variable=Head Tilt Right}
   &RigUnit_ModifyTransforms(
    ItemToModify=((Item=(Type=Bone,Name="dyn_hair_side_l_base"),Transform=(Rotation=(X=0,Y=0,Z=0,W=1),Translation=(X=0,Y=0,Z=0),Scale3D=(X=1,Y=1,Z=1))))
    ItemToModify.0.Transform.Rotation=&RigVMFunction_MathQuaternionFromEuler(Euler=(X=0,Y=0,Z=0)).Result
    ExecutePin=@SequenceAggregate_1.A
    Weight=@n1.Value
   )
   > &RigUnit_ModifyTransforms(
    ItemToModify=((Item=(Type=Bone,Name="dyn_hair_bk_mid_base"),Transform=(Rotation=(X=0,Y=0,Z=0,W=1),Translation=(X=0,Y=0,Z=0),Scale3D=(X=1,Y=1,Z=1))))
    ItemToModify.0.Transform.Rotation=&RigVMFunction_MathQuaternionFromEuler(Euler=(X=0,Y=10,Z=0)).Result
    Weight=@n1.Value
   )
   > &RigUnit_ModifyTransforms(
    ItemToModify=((Item=(Type=Bone,Name="dyn_hair_side_r_base"),Transform=(Rotation=(X=0,Y=0,Z=0,W=1),Translation=(X=2,Y=0,Z=0),Scale3D=(X=1,Y=1,Z=1))))
    ItemToModify.0.Transform.Rotation=&RigVMFunction_MathQuaternionFromEuler(Euler=(X=0,Y=-10,Z=-20)).Result
    Weight=@n1.Value
   )
  ]}
]
hierarchy [
 root {loc=<0 0 0>, rot=<0 0 0 1>}
 pelvis {parent=root, loc=<0 -0.5301 98.0571>, rot=<0.0127 0.707 -0.0127 -0.707>}
 spine_01 {parent=pelvis, loc=<6.1208 -0.9328 0>, rot=<0 0 0.034 -0.9994>}
 spine_02 {parent=spine_01, loc=<6.5996 -0.2095 0>, rot=<0 0 0.0238 -0.9997>}
 spine_03 {parent=spine_02, loc=<6.4914 -0.1237 0>, rot=<0 0 0.0149 -0.9999>}
 spine_04 {parent=spine_03, loc=<6.9598 0.0794 0>, rot=<0 0 -0.0132 -0.9999>}
 spine_05 {parent=spine_04, loc=<6.9987 0 0>, rot=<0 0 -0.0015 -1>}
 neck_01 {parent=spine_05, loc=<11.3138 0.1529 0>, rot=<0 0 -0.2003 -0.9797>}
 neck_02 {parent=neck_01, loc=<4.5627 0 0>, rot=<0 0 0.0504 -0.9987>}
 head {parent=neck_02, loc=<5.3914 0 0>, rot=<0 0 0.115 -0.9934>}
 dyn_hair_frnt_mid_parent {parent=head, loc=<11.7778 7.4085 -2.7887>, rot=<0.0308 -0.8904 -0.3418 -0.2991>}
 dyn_hair_frnt_mid_01 {parent=dyn_hair_frnt_mid_parent, loc=<3.8918 0 0>, rot=<0 -0.1829 0 -0.9831>}
 dyn_hair_frnt_mid_02 {parent=dyn_hair_frnt_mid_01, loc=<4.0302 0 0>, rot=<0 0.6122 0 -0.7907>}
 dyn_hair_side_frnt_l {parent=head, loc=<3.0389 1.7125 -7.0997>, rot=<0.0041 -0.9972 0.0003 0.0751>}
 dyn_hair_side_l_base {parent=head, loc=<11.1645 -3.0643 -7.4279>, rot=<0.1193 -0.9834 0.1082 -0.083>}
 dyn_hair_side_l_parent {parent=dyn_hair_side_l_base, loc=<0 0 0>, rot=<0 0 0 1>}
 dyn_hair_side_l_01 {parent=dyn_hair_side_l_parent, loc=<9.5158 0 -1.1108>, rot=<0 -0.0898 0 -0.996>}
 dyn_hair_bk_mid_base {parent=head, loc=<11.4079 -8.0473 0>, rot=<0.0786 -0.9969 0 0>}
 dyn_hair_bk_mid_parent {parent=dyn_hair_bk_mid_base, loc=<0 0 0>, rot=<0 0 0 1>}
 dyn_hair_bk_mid_01 {parent=dyn_hair_bk_mid_parent, loc=<8.6603 0.3842 0>, rot=<0 0 -0.0873 -0.9962>}
 dyn_hair_side_r_base {parent=head, loc=<11.0321 -4.215 7.1468>, rot=<0.9718 0.0556 -0.0215 -0.2282>}
 dyn_hair_side_r_parent {parent=dyn_hair_side_r_base, loc=<0 0 0>, rot=<0 0 0 1>}
 dyn_hair_side_r_01 {parent=dyn_hair_side_r_parent, loc=<-9.2576 0.0001 0.032>, rot=<0 -0.0094 0 -1>}
 dyn_hair_side_frnt_r_base {parent=head, loc=<8.887 3.7233 7.6414>, rot=<0.9989 0.0406 -0.0207 -0.0139>}
 dyn_hair_side_frnt_r_parent {parent=dyn_hair_side_frnt_r_base, loc=<0 0 0>, rot=<0 0 0 1>}
 dyn_hair_side_frnt_r_01 {parent=dyn_hair_side_frnt_r_parent, loc=<-9.1105 0.0001 0.0337>, rot=<0 -0.0094 0 -1>}
 dyn_hair_frnt_r_parent {parent=dyn_hair_side_frnt_r_parent, loc=<-6.758 -2.7135 0.3698>, rot=<0.2302 -0.0413 -0.0286 -0.9718>}
 dyn_hair_frnt_r_01 {parent=dyn_hair_frnt_r_parent, loc=<-5.1649 0.0001 0.0277>, rot=<0 -0.0094 0 -1>}
 clavicle_l {parent=spine_05, loc=<5.4831 -0.1053 -3.2946>, rot=<0.0465 -0.7289 -0.009 -0.6829>}
 clavicle_r {parent=spine_05, loc=<5.4831 -0.1053 3.2946>, rot=<0.7289 0.0465 -0.6829 0.009>}
]
variables [Head Tilt Right {type=double, default=0}]
