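# Physics asset for the face accessory mesh of the female medium RoosterMelt cosmetic body, setting up physics for its dynamic hair bones.
# Capsules on head, neck, clavicles and upper spine provide body shapes, including an oversized second neck capsule.
# Back, front and side hair chains use tight swing limits and minimal twist.
# Chains root on the head or tiny base capsules, except front-right, which branches from side-front-right.
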PhysicsAsset /BRCosmetics/Characters/Player/Female/Medium/Bodies/F_MED_RoosterMelt/Meshes/Parts/F_MED_RoosterMelt_FaceAcc_Physics
bodies {
 clavicle_l [{shapeType=Capsule, length=9.9, radius=5.4, shapeName=clavicle_l_capsule, center=<5 0 0>, rotation=<rot 0 -90 90>}]
 clavicle_r [{shapeType=Capsule, length=9.9, radius=5.4, shapeName=clavicle_r_capsule, center=<-5 0 0>, rotation=<rot 0 90 90>}]
 dyn_hair_bk_mid_01 [{shapeType=Capsule, length=4, radius=3, shapeName=dyn_hair_bk_mid_01_capsule, center=<4 0 0>, rotation=<rot 0 90 90>}]
 dyn_hair_bk_mid_base [{shapeType=Capsule, length=0.505, radius=0.505, shapeName=dyn_hair_bk_mid_base_capsule, rotation=<rot 0 -90 0>}]
 dyn_hair_bk_mid_parent [{shapeType=Capsule, length=5, radius=2, shapeName=dyn_hair_bk_mid_parent_capsule, center=<4 0 0>, rotation=<rot 0 90 90>}]
 dyn_hair_frnt_mid_01 [{shapeType=Capsule, length=5, radius=2, shapeName=dyn_hair_frnt_mid_01_capsule, rotation=<rot 0 90 90>}]
 dyn_hair_frnt_mid_02 [{shapeType=Capsule, length=5, radius=2, shapeName=dyn_hair_frnt_mid_02_capsule, rotation=<rot 0 90 0>}]
 dyn_hair_frnt_mid_parent [{shapeType=Capsule, length=5, radius=2, shapeName=dyn_hair_frnt_mid_parent_capsule, rotation=<rot 0 90 90>}]
 dyn_hair_frnt_r_01 [{shapeType=Capsule, length=5, radius=2, shapeName=dyn_hair_frnt_r_01_capsule, center=<-4 0 0>, rotation=<rot 0 90 90>}]
 dyn_hair_frnt_r_parent [{shapeType=Capsule, length=5, radius=2, shapeName=dyn_hair_frnt_r_parent_capsule, center=<-4 0 0>, rotation=<rot 0 90 90>}]
 dyn_hair_side_frnt_l [{shapeType=Capsule, length=5, radius=2, shapeName=dyn_hair_side_frnt_l_capsule, rotation=<rot 0 90 90>}]
 dyn_hair_side_frnt_r_01 [{shapeType=Capsule, length=5, radius=2, shapeName=dyn_hair_side_frnt_r_01_capsule, center=<-4 0 0>, rotation=<rot 0 90 90>}]
 dyn_hair_side_frnt_r_base [{shapeType=Capsule, length=0.505, radius=0.505, shapeName=dyn_hair_side_frnt_r_base_capsule, rotation=<rot 0 -90 0>}]
 dyn_hair_side_frnt_r_parent [{shapeType=Capsule, length=5, radius=2, shapeName=dyn_hair_side_frnt_r_parent_capsule, center=<-4 0 0>, rotation=<rot 0 90 90>}]
 dyn_hair_side_l_01 [{shapeType=Capsule, length=4, radius=3, shapeName=dyn_hair_side_l_01_capsule, center=<4 0 0>, rotation=<rot 0 90 90>}]
 dyn_hair_side_l_base [{shapeType=Capsule, length=0.505, radius=0.505, shapeName=dyn_hair_side_l_base_capsule, rotation=<rot 0 -90 0>}]
 dyn_hair_side_l_parent [{shapeType=Capsule, length=5, radius=2, shapeName=dyn_hair_side_l_parent_capsule, center=<4 0 0>, rotation=<rot 0 90 90>}]
 dyn_hair_side_r_01 [{shapeType=Capsule, length=4, radius=2.5, shapeName=dyn_hair_side_r_01_capsule, center=<-3.0628 0 0>, rotation=<rot 0 90 90>}]
 dyn_hair_side_r_base [{shapeType=Capsule, length=0.505, radius=0.505, shapeName=dyn_hair_side_r_base_capsule, rotation=<rot 0 -90 0>}]
 dyn_hair_side_r_parent [{shapeType=Capsule, length=5, radius=2, shapeName=dyn_hair_side_r_parent_capsule, center=<-4 0 0>, rotation=<rot 0 0 90>}]
 head [{shapeType=Capsule, length=4, radius=7.2817, center=<4.5949 0.4221 0>, rotation=<rot -90 0 -0.0007>}]
 neck_01 [
  {shapeType=Capsule, length=8, radius=5.1, shapeName=neck_01_capsule, rotation=<rot 0 -90.0003 90>}
  {shapeType=Capsule, length=22, radius=18, shapeName=neck_01_capsule, center=<-7.945 13.3265 0.2787>}
 ]
 spine_05 [{shapeType=Capsule, length=8, radius=8.7, shapeName=spine_05_capsule, center=<2.92 0.38 0>, rotation=<rot 0 -90 0>}]
}
constraints [
 {bone1Name=dyn_hair_bk_mid_parent, bone2Name=dyn_hair_bk_mid_base, swing1LimitDegrees=20, swing2LimitDegrees=10, twistLimitDegrees=2}
 {bone1Name=dyn_hair_bk_mid_01, bone2Name=dyn_hair_bk_mid_parent, swing1LimitDegrees=20, swing2LimitDegrees=20, twistLimitDegrees=2}
 {bone1Name=dyn_hair_frnt_mid_parent, bone2Name=head, swing1LimitDegrees=5, swing2LimitDegrees=10, twistLimitDegrees=0}
 {bone1Name=dyn_hair_frnt_mid_01, bone2Name=dyn_hair_frnt_mid_parent, swing1LimitDegrees=10, swing2LimitDegrees=20, twistLimitDegrees=0}
 {bone1Name=dyn_hair_frnt_mid_02, bone2Name=dyn_hair_frnt_mid_01, swing1LimitDegrees=30, swing2LimitDegrees=10, twistLimitDegrees=0}
 {bone1Name=dyn_hair_frnt_r_01, bone2Name=dyn_hair_frnt_r_parent, swing1LimitDegrees=20, swing2LimitDegrees=10, twistLimitDegrees=0}
 {bone1Name=dyn_hair_side_frnt_l, bone2Name=head, swing1LimitDegrees=20, swing2LimitDegrees=20, twistLimitDegrees=0}
 {
  bone1Name=dyn_hair_side_frnt_r_parent
  bone2Name=dyn_hair_side_frnt_r_base
  swing1LimitDegrees=20
  swing2LimitDegrees=10
  twistLimitDegrees=2
 }
 {bone1Name=dyn_hair_side_frnt_r_01, bone2Name=dyn_hair_side_frnt_r_parent, swing1LimitDegrees=20, swing2LimitDegrees=10, twistLimitDegrees=2}
 {bone1Name=dyn_hair_side_l_parent, bone2Name=dyn_hair_side_l_base, swing1LimitDegrees=20, swing2LimitDegrees=10, twistLimitDegrees=2}
 {bone1Name=dyn_hair_side_l_01, bone2Name=dyn_hair_side_l_parent, swing1LimitDegrees=20, swing2LimitDegrees=10, twistLimitDegrees=2}
 {bone1Name=dyn_hair_side_r_parent, bone2Name=dyn_hair_side_r_base, swing1LimitDegrees=20, swing2LimitDegrees=10, twistLimitDegrees=2}
 {bone1Name=dyn_hair_side_r_01, bone2Name=dyn_hair_side_r_parent, swing1LimitDegrees=20, swing2LimitDegrees=20, twistLimitDegrees=5}
 {bone1Name=dyn_hair_frnt_r_parent, bone2Name=dyn_hair_side_frnt_r_parent, swing1LimitDegrees=20, swing2LimitDegrees=10, twistLimitDegrees=2}
]
